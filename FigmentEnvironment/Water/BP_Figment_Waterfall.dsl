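; Waterfall environment actor whose construction script configures water and terrain materials, splash VFX, base sounds and a bottom mesh from editable controls.
; The Cascade splash and stereo sounds follow preset or manual foam heights, with particles held at unit scale and sized by actor X scale.
; Niagara receives world-space top and bottom positions plus tuning, while the bottom mesh gets shape and color material parameters.
; BeginPlay reapplies the Niagara rotation after a delay.
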(event EventBeginPlay
  ; Just a double check
  (Utilities|FlowControl|Delay 2e+01)
  (Transformation|SetRelativeRotation (Variables|Default|GetNS_WaterfallFX) "0.000000,180.000000,0.000000" false false))

(fn ConstructionScript
  (bind _foam_position (Variables|ParticleNotEdit|GetFoam%20Position))
  (bind _foam_height_postion_manual (Variables|ParticleNotEdit|GetFoam%20Height%20Postion_Manual))
  (bind _foam_height_postion_04 (Variables|ParticleNotEdit|GetFoam%20Height%20Postion%2004))
  (bind _foam_height_postion_03 (Variables|ParticleNotEdit|GetFoam%20Height%20Postion%2003))
  (bind _foam_height_postion_02 (Variables|ParticleNotEdit|GetFoam%20Height%20Postion%2002))
  (bind _foam_height_postion_01 (Variables|ParticleNotEdit|GetFoam%20Height%20Postion%2001))
  (bind _returnvalue (Utilities|Select :Index _foam_position :Option%200 _foam_height_postion_01 :Option%201 _foam_height_postion_02 :Option%202 _foam_height_postion_03 :Option%203 _foam_height_postion_04 :Option%204 _foam_height_postion_manual))
  (bind _foam_depth_adjustment (Variables|ParticleControls|GetFoam%20Depth%20Adjustment))
  (bind _returnvalue_2 (Math|Vector|MakeVector 512.0 _foam_depth_adjustment _returnvalue))
  (bind _returnvalue_3 (Transformation|GetActorScale3D))
  (bind _foam_manual_height_placement (Variables|ParticleControls|GetFoam%20Manual%20Height%20Placement))
  (bind _returnvalue_4 (Transformation|GetActorTransform))
  (bind _returnvalue_5 (Math|Transform|TransformLocation _returnvalue_4 _foam_manual_height_placement))
  (bind _x (.x _returnvalue_3))
  (bind _x_2 (.x _returnvalue_5))
  (bind _y (.y _returnvalue_5))
  (bind _returnvalue_6 (Transformation|GetActorLocation))
  (bind _z (.z _returnvalue_6))
  (bind _returnvalue_7 (Math|Vector|MakeVector _x_2 _y _z))
  (bind _returnvalue_8 (Math|Float|Absolute(Float) _x))
  (bind _sm_waterfall_meshbottom (Variables|Default|GetSM_Waterfall_MeshBottom))
  (bind _ns_waterfallfx (Variables|Default|GetNS_WaterfallFX))
  (bind _watermaterialoverride (Variables|Default|GetWaterMaterialOverride))
  (bind _ns_waterfallfx_2 (Variables|Default|GetNS_WaterfallFX))
  (bind _waterfallvfx (Variables|Default|GetWaterFallVFX))
  (bind _sm_waterfall_meshbottom_2 (Variables|Default|GetSM_Waterfall_MeshBottom))
  (|CallParentFunction)
  (Utilities|FlowControl|Sequence
    (:then_0
      (Effects|Components|ParticleSystem|SetTemplate (Variables|Default|GetWaterFallVFX) "/Game/Effects/Fort_Effects/Effects/Water/P_WaterFall_New.P_WaterFall_New")
      (Rendering|Material|SetMaterial (Variables|Default|GetStaticMesh1) 0 (Variables|TerrainMesh|GetTerrain%20Material))
      (Utilities|FlowControl|Sequence
        (:then_0
          (Transformation|SetRelativeLocation (Variables|Default|GetWaterFall_Base_Sound_L) "300.000000, 0, 0" false false)
          (Transformation|SetRelativeLocation (Variables|Default|GetWaterFall_Base_Sound_R) "-300.000000, 0, 0" false false)
          (Audio|Components|Audio|SetSound (Variables|Default|GetWaterFall_Base_Sound_L) (Variables|Default|GetWaterfall_Sound_L_Cue))
          (Audio|Components|Audio|SetSound (Variables|Default|GetWaterFall_Base_Sound_R) (Variables|Default|GetWaterfall_Sound_R_Cue))
          (Effects|Components|ParticleSystem|SetVectorParameter (Variables|Default|GetWaterFallVFX) (Variables|ParticleNotEdit|GetFoamColor) (Math|Conversions|ToVector(LinearColor) (Variables|ParticleControls|GetSplashFX_Color)))
          (if (Variables|ParticleControls|GetUse_SplashFX_Height_01)
            (Variables|ParticleNotEdit|SetFoam%20Position 0)
            (elif (Variables|ParticleControls|GetUse_SplashFX_Height_02)
              (Variables|ParticleNotEdit|SetFoam%20Position 1)
              (elif (Variables|ParticleControls|GetUse_SplashFX_Height_03)
                (Variables|ParticleNotEdit|SetFoam%20Position 2)
                (Variables|ParticleNotEdit|SetFoam%20Height%20Postion_Manual (.z (Variables|ParticleControls|GetFoam%20Manual%20Height%20Placement)))
                ;[1 1 1] Foam Positioning
                (Transformation|SetRelativeLocationAndRotation (Variables|Default|GetWaterFallVFX) _returnvalue_2 (Variables|ParticleControls|GetSplashFX_Rotation) false false)
                (Transformation|SetRelativeLocation (Variables|Default|GetWaterFall_Base_Sound_L) (Math|Vector|MakeVector (+ (.x _returnvalue_2) (Variables|Default|GetAudio_LR_Offset)) (.y _returnvalue_2) (+ (.z _returnvalue_2) (Variables|ParticleControls|GetAudio%20Z%20offset))) false false)
                (Transformation|SetRelativeLocation (Variables|Default|GetWaterFall_Base_Sound_R) (Math|Vector|MakeVector (- (.x _returnvalue_2) (Variables|Default|GetAudio_LR_Offset)) (.y _returnvalue_2) (+ (.z _returnvalue_2) (Variables|ParticleControls|GetAudio%20Z%20offset))) false false)
                ;[1 1 1] Insures Particle Match Scaled Width of Waterfall
                (Transformation|SetWorldScale3D (Variables|Default|GetWaterFallVFX) "1.000000,1.000000,1.000000")
                (Effects|Components|ParticleSystem|SetVectorParameter _waterfallvfx "SizeScale" (Math|Conversions|ToVector(Float) _returnvalue_8))
                (Effects|Components|ParticleSystem|SetFloatParameter _waterfallvfx "CylinderSize" (* 4.5e+02 _returnvalue_8))
                (Effects|Components|ParticleSystem|SetFloatParameter _waterfallvfx "CylinderSize_Rolling" (* 2.5e+02 _returnvalue_8))
                (Effects|Components|ParticleSystem|SetFloatParameter _waterfallvfx "SpawnRate" _returnvalue_8)
                (elif (Variables|ParticleControls|GetUse_SplashFX_Height_04)
                  (Variables|ParticleNotEdit|SetFoam%20Position 3)
                  (else
                    (Variables|ParticleNotEdit|SetFoam%20Position 4)))))))
        (:then_1
          (Utilities|IsValid _watermaterialoverride
            (:Is%20Valid
              (Rendering|Material|SetMaterial (Variables|Default|GetStaticMesh1) 1 _watermaterialoverride))))))
    (:then_1
      (Transformation|SetWorldScale3D _ns_waterfallfx "1.000000,1.000000,1.000000")
      (Transformation|SetRelativeRotation _ns_waterfallfx "0.000000,180.000000,0.000000" false false)
      (Niagara|SetNiagaraVariable(Position) _ns_waterfallfx_2 "TopPosition" (+ _returnvalue_7 (Math|Transform|TransformDirection _returnvalue_4 (Variables|NiagaraControls|GetTop%20Pos%20Offset))))
      (Niagara|SetNiagaraVariable(Position) _ns_waterfallfx_2 "BottomPosition" (Utilities|Operators|Add _returnvalue_7 (Math|Transform|TransformDirection _returnvalue_4 (Variables|NiagaraControls|GetBottom%20Pos%20Offset)) (Math|Vector|MakeVector 0.0 0.0 (- (.z _returnvalue_5) _z))))
      (Effects|Components|ParticleSystem|SetFloatParameter _ns_waterfallfx_2 "RippleOffset" (Variables|NiagaraControls|GetRipple%20Offset))
      (Effects|Components|ParticleSystem|SetFloatParameter _ns_waterfallfx_2 "fallwidth" (Variables|NiagaraControls|GetFallWidth))
      (Effects|Components|ParticleSystem|SetFloatParameter _ns_waterfallfx_2 "BottomSmokeDensity" (Variables|NiagaraControls|GetBottom%20Smoke%20Density))
      (Effects|Components|ParticleSystem|SetFloatParameter _ns_waterfallfx_2 "HitSplashEffectDensity" (Variables|NiagaraControls|GetHitsplashEffectDencity))
      (Effects|Components|ParticleSystem|SetFloatParameter _ns_waterfallfx_2 "RainbowStrength" (Variables|NiagaraControls|GetRainbow%20Str))
      (Effects|Components|ParticleSystem|SetFloatParameter _ns_waterfallfx_2 "RippleOpacity" (Variables|NiagaraControls|GetRipple%20Opac))
      (Effects|Components|ParticleSystem|SetFloatParameter _ns_waterfallfx_2 "RippleWidth" (Variables|NiagaraControls|GetRipple%20Width))
      (Niagara|NiagaraSetPositionArray _ns_waterfallfx_2 "SpawnSplashPositionArray" (Variables|NiagaraControls|GetNiagara%20Spawn%20Splash%20Pos%20Array))
      (Effects|Components|ParticleSystem|SetBoolParameter _ns_waterfallfx_2 "UseIdMask" (Variables|NiagaraControls|GetRippleUseIDmask))
      (Effects|Components|ParticleSystem|SetFloatParameter _ns_waterfallfx_2 "Bottom SPlash Scale" (Variables|NiagaraControls|GetBottom%20Sprite%20scale))
      (Effects|Components|ParticleSystem|SetFloatParameter _ns_waterfallfx_2 "TopSplashRate" (Variables|NiagaraControls|GetTop%20Splash%20Rate))
      (Effects|Components|ParticleSystem|SetFloatParameter _ns_waterfallfx_2 "TopSplashScale" (Variables|NiagaraControls|GetTopSplashScale))
      (Effects|Components|ParticleSystem|SetFloatParameter _ns_waterfallfx_2 "bottomsplashrate" (Variables|NiagaraControls|GetBottom%20Spalsh%20Rate))
      (Effects|Components|ParticleSystem|SetColorParameter _ns_waterfallfx_2 "BottomSplash Color" (Variables|NiagaraControls|GetBottom%20Splash%20Color)))
    (:then_2
      (Transformation|SetRelativeLocation _sm_waterfall_meshbottom_2 (+ (Variables|ParticleControls|GetFoam%20Manual%20Height%20Placement) 0) false false)
      (bind _output_get (Variables|Default|SetMeshBottom_Scale_TMP (Variables|/Script/Engine.SceneComponent|GetRelativeScale3D _sm_waterfall_meshbottom_2)))
      (Transformation|SetRelativeScale3D _sm_waterfall_meshbottom_2 (* _output_get (Variables|Default|GetMesh%20Bottom%20Scale)))
      (Rendering|Material|SetScalarParameterValueonMaterials _sm_waterfall_meshbottom "BPInput_Width" (Variables|Default|GetWidth))
      (Rendering|Material|SetScalarParameterValueonMaterials _sm_waterfall_meshbottom "BPInput_Height" (Variables|Default|GetHeight))
      (Rendering|Material|SetScalarParameterValueonMaterials _sm_waterfall_meshbottom "BPInput_EdgeOffset" (Variables|Default|GetEdgeOffset))
      (Rendering|Material|SetScalarParameterValueonMaterials _sm_waterfall_meshbottom "BPInput_BevelLength" (Variables|Default|GetBevelLength))
      (Rendering|Material|SetScalarParameterValueonMaterials _sm_waterfall_meshbottom "BPInput_ExtendLength" (Variables|Default|GetExtendLength))
      (Rendering|Material|SetScalarParameterValueonMaterials _sm_waterfall_meshbottom "BPInput_SplashSize" (Variables|Default|GetSplashSize))
      (Rendering|Material|SetVectorParameterValueonMaterials _sm_waterfall_meshbottom "BPInput_BPPosition" (Transformation|GetWorldLocation (Variables|Default|GetSM_Waterfall_MeshBottom)))
      (Rendering|Material|SetColorParameterValueonMaterials _sm_waterfall_meshbottom "BaseColor" (Variables|Default|GetMesh%20Bottom%20BaseColor))
      (Rendering|Material|SetColorParameterValueonMaterials _sm_waterfall_meshbottom "BaseColor_Edge" (Variables|Default|GetMesh%20Bottom%20BaseColor_Edge)))))
