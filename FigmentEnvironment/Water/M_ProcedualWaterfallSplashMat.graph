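# Masked subsurface material for a procedural waterfall splash, driven by position and size parameters fed from a Blueprint.
# Radial UVs shape edge masks for color blending and dithered cutout, which also varies with camera height relative to the waterfall.
# Two panning Perlin normal maps and vertex normals build splash displacement, scaled by splash size.
# Further chains compute noise breakup, camera depth fade, dithered pixel depth offset and time-of-day edge subsurface tint.
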Material /FigmentEnvironment/Water/M_ProcedualWaterfallSplashMat
graph {
 properties {
  blendMode=BLEND_Masked
  pixelDepthOffsetMode=PDOM_Legacy
  previewMesh=&/SM_Waterfall_MeshBottom
  shadingModel=MSM_Subsurface
  usedShadingModels=Subsurface
 }
 nodes [
  Constant() {r=1}
  Constant() {r=1}
  n1 = VectorParameter() {parameterName=MaskEdgeWidth, defaultValue=<rgba 0.6 0.2 0.5 1>}
  n2 = TextureCoordinate() {uTiling=2, vTiling=2}
  n3 = MaterialFunctionCall(Float2=MaterialFunctionCall(Vector or UVs=Add(A=@n2) {constB=-1}) {materialFunction=&VectorToRadialValue}.Radial Coordinates) {materialFunction=&BreakOutFloat2Components}
  n4 = OneMinus(None=@n3.G)
  n5 = Subtract(A=VectorParameter() {parameterName=BPInput_BPPosition, defaultValue=<rgba 0 0 0 1>}.RGB, B=CameraPositionWS())
  n6 = Divide(A=Add(A=MaterialFunctionCall(Float3=@n5) {materialFunction=&BreakOutFloat3Components}.B) {constB=-60}) {constB=40}
  n7 = LinearInterpolate(B=Power(Base=SmoothStep(Max=@n1.G, Value=@n4)) {constExponent=1.6}, Alpha=OneMinus(None=Saturate(None=@n6))) {constAlpha=1, constB=0}
  n8 = Constant() {r=1}
  n9 = MaterialFunctionCall(Alpha Threshold=@n7) {materialFunction=&DitherTemporalAA_Fortnite}
  If(A=@n7, "A > B"=@n8, "A == B"=@n8, "A < B"=@n9.Result) {constB=0.8}
  LinearInterpolate(
   A=VectorParameter() {parameterName=BaseColor_Edge, defaultValue=<rgba 0 0.1 0.0833 0.1>}.RGB
   B=VectorParameter() {parameterName=BaseColor, defaultValue=<rgba 1 1 1 0.1>}.RGB
   Alpha=Power(Base=SmoothStep(Max=@n1.B, Value=@n4), Exp=ScalarParameter() {defaultValue=2, parameterName=EdgeColorBlendPower})
  )
  n10 = TextureCoordinate()
  n11 = Constant() {r=0.5}
  n12 = Power(Base=OneMinus(None=SphereMask(A=@n10, B=@n11) {attenuationRadius=0.5, hardnessPercent=25}))
  MaterialFunctionCall(Alpha Threshold=@n12) {materialFunction=&DitherTemporalAA}
  n13 = TextureObject() {
   samplerType=SAMPLERTYPE_Normal
   texture=/Game/Athena/Artemis/Environments/BuildingActors/WaterFall/Materials/Texture/T_Lowres_Simple_Perlin_N
  }
  n14 = VectorParameter() {group=GroupA, parameterName=GroupA_UVControl, channelNames {a=VPan, b=UPan, g=VScale, r=UScale}, defaultValue=<rgba 0.2 0.5 0 -0.62>}
  n15 = ScalarParameter() {defaultValue=1, parameterName=CustomParamter_SplashUVScale}
  n16 = TextureCoordinate() {uTiling=2, vTiling=2}
  n17 = Multiply(
   A=MaterialFunctionCall(Vector or UVs=Add(A=@n16) {constB=-1}) {materialFunction=&VectorToRadialValue}.Radial Coordinates
   B=AppendVector(A=Constant() {r=1}, B=@n15)
  )
  n18 = TextureObject() {
   samplerType=SAMPLERTYPE_Normal
   texture=&T_Lowres_Simple_Perlin_N
  }
  n19 = VectorParameter() {group=GroupA2, parameterName=GroupA2_UVControl, channelNames {a=VPan, b=UPan, g=VScale, r=UScale}, defaultValue=<rgba 0.3 1 0 -0.407>}
  n20 = Multiply(A=@n19.VPan, B=@n15)
  n21 = MaterialFunctionCall(
   BaseNormal=MaterialFunctionCall(Texture=@n13, UVControl_RGB=@n14, UVControl_A=Multiply(A=@n14.VPan, B=@n15), "[Override]UVInput"=@n17) {materialFunction=&FXCommon_TextureSample}.RGB
   AdditionalNormal=MaterialFunctionCall(
    Normal=MaterialFunctionCall(Texture=@n18, UVControl_RGB=@n19, UVControl_A=@n20, "[Override]UVInput"=@n17) {materialFunction=&FXCommon_TextureSample}.RGB
    Flatness=ScalarParameter() {defaultValue=0.5, parameterName=Flata}
   ) {materialFunction=&FlattenNormal}.Result
  ) {materialFunction=&BlendAngleCorrectedNormals}
  MaterialFunctionCall(Normal=@n21.Result, Flatness=ScalarParameter() {defaultValue=0.8, parameterName=t}) {materialFunction=&FlattenNormal}
  n22 = ScalarParameter() {parameterName=BPInput_SplashSize}
  n23 = Multiply(
   A=Power(Base=OneMinus(None=@n3.G), Exp=ScalarParameter() {defaultValue=2, parameterName=CentA1})
   B=Multiply(A=VertexNormalWS(), B=ScalarParameter() {defaultValue=130, parameterName=CentA2})
  )
  n24 = VectorParameter() {group=03.WPO, parameterName=FromNoiseNormalB_Multiply, defaultValue=<rgba 0 0 264.3313 -0.5>}
  n25 = Multiply(A=Power(Base=@n4) {constExponent=1.2}, B=SmoothStep(Max=@n1.R, Value=@n3.G))
  n26 = Multiply(A=Multiply(A=Add(A=@n21.Result, B=@n24.A) {constB=-0.5}, B=@n24.RGB), B=@n25)
  Multiply(
   A=Multiply(A=Divide(A=@n22) {constB=900}, B=Add(A=@n23, B=@n26))
   B=ScalarParameter() {defaultValue=1, parameterName=CustomParameter_SplashWPOScale}
  )
  n27 = Distance(A=TextureCoordinate(), B=Constant() {r=0.5})
  Multiply(
   A=LinearInterpolate(A=ScalarParameter() {parameterName=PDO}, B=ScalarParameter() {parameterName=PDOouter}, Alpha=Saturate(None=@n27))
   B=MaterialFunctionCall(Alpha Threshold=Constant() {r=1}) {materialFunction=&DitherTemporalAA}.Result
  )
  n28 = TextureCoordinate() {uTiling=0.5, vTiling=0.5}
  n29 = Panner(Coordinate=@n28, Speed=Constant() {r=0.1}) {speedX=0.5, speedY=0.5}
  n30 = Panner(Coordinate=@n28, Speed=Constant() {r=-0.1}) {speedX=-0.5, speedY=-0.5}
  n31 = Divide(A=Add(
    A=TextureSample(UVs=@n29) {samplerType=SAMPLERTYPE_LinearColor, texture=&HighContrastNoise}.R
    B=TextureSample(UVs=@n30) {samplerType=SAMPLERTYPE_LinearColor, texture=&HighContrastNoise}.G
   ))
  n32 = Saturate(None=MaterialFunctionCall(In=Subtract(A=@n31) {constA=0.5, constB=-0.1}, Contrast=ScalarParameter() {defaultValue=1, parameterName=noise Contrast}) {materialFunction=&CheapContrast}.Result)
  n33 = LinearInterpolate(A=@n32, Alpha=@n12)
  n34 = ScalarParameter() {parameterName="Fade Length (S)"}
  n35 = ScalarParameter() {parameterName="Fade Offset (S)"}
  n36 = Multiply(
   A=@n9.Result
   B=MaterialFunctionCall(Alpha Threshold=MaterialFunctionCall(Fade Length=@n34, Fade Offset=@n35) {materialFunction=&CameraDepthFade}.Result) {materialFunction=&DitherTemporalAA}.Result
  )
  Multiply(A=Multiply(A=Multiply(A=@n33, B=@n36)) {constB=2}) {constB=0.9}
  Multiply(
   A=Multiply(A=@n25, B=VectorParameter() {parameterName=EdgeSubsurfaceColor, defaultValue=<rgba 0.0375 0.05 0.025 1>}.RGB)
   B=MaterialFunctionCall(
    Color_Morning=Constant3Vector() {constant=<rgba 0.3 0.1875 0.165 1>}
    Color_Day=Constant3Vector() {constant=<rgba 1.2 1.2 1.2 1>}
    Color_Evening=Constant3Vector() {constant=<rgba 0.1 0.05 0.04 1>}
    Color_Night=Constant3Vector() {constant=<rgba 0.3 0.4 0.5 1>}
   ) {materialFunction=&Color_Time_of_Day}.Result
  )
  OneMinus(None=@n32)
  TextureCoordinate()
  VectorParameter() {parameterName=Sub, defaultValue=<rgba 0.1 0.1 0.1 0.5>}
 ]
}
