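# Translucent lit Niagara sprite material for radial ripples at waterfalls, with edge-to-center color tinted by particle color.
# Panning noise distorts radial UVs that sample ripple textures for a flattened normal and contrast-shaped opacity, faded at center and edges and by particle alpha.
# An eight-tap, noise-dithered shading model ID test (10, single-layer water) softly masks ripples to water on deferred, bypassed on forward and mobile.
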Material /FigmentEnvironment/Water/M_Waterfall_Ripple
graph {
 properties {
  bUsedWithNiagaraSprites=true
  blendMode=BLEND_Translucent
  pixelDepthOffsetMode=PDOM_Legacy
  translucencyLightingMode=TLM_Surface
  usedShadingModels=Default Lit
 }
 nodes [
  Constant() {r=1}
  Constant() {r=1}
  n1 = Custom!(None) {
   code=View.TemporalAAParams
   desc="\n*** Original source before expression upgrade ***\nView.TemporalAAParams\n*** Original source before expression upgrade ***\nFrame.TemporalAAParams"
   description=TemporalAA SampleID
   outputType=CMOT_Float2
  }
  MaterialExpressionAdd_19 = Add(
   A=MaterialFunctionCall!(TextureResolution=Constant2Vector() {g=1, r=1}) {materialFunction=/Engine/Functions/Engine_MaterialFunctions02/Texturing/ScreenAlignedPixelToPixelUVs}.Result
   B=ComponentMask(None=@n1) {r=true}
  )
  n2 = Custom!(p=@MaterialExpressionAdd_19) {
   code="((uint)(p.x) + 2 * (uint)(p.y)) % 5"
   outputType=CMOT_Float1
   inputs [{inputName=p, input {expression=&self:MaterialExpressionAdd_19}}]
  }
  Divide(A=@n2) {constB=6}
  n3 = ParticleColor()
  MaterialFunctionCall(Emissive=@n3.RGB) {materialFunction=&MF_FX_EmissiveAdjustments}
  n4 = TextureObjectParameter() {
   group=RippleTexture
   parameterName=RippleTexture_NTexture
   samplerType=SAMPLERTYPE_Normal
   sortPriority=0
   texture=/Game/Athena/Apollo/Water/Textures/02_Rounded/03_WaterFoam_AccentSwirls_N
  }
  n5 = TextureCoordinate()
  n6 = VectorParameter() {
   group=RippleTexture
   parameterName=RippleTexture_UVControl
   channelNames {a=VPan, b=UPan, g=VScale, r=UScale}
   defaultValue=<rgba 0.3 1.5 0.01 -0.1>
  }
  n7 = TextureCoordinate()
  n8 = VectorParameter() {group=21.NoiseB, parameterName=NoiseB_ScaleAndOffset, defaultValue=<rgba 0.8 0.8 -0.02 0.07>}
  n9 = AppendVector(A=@n8.R, B=@n8.R)
  n10 = AppendVector(A=@n8.B, B=@n8.B)
  n11 = Time() {bOverride_Period=true, period=6000}
  n12 = TextureSample(UVs=Add(A=Multiply(A=@n7, B=@n9), B=Multiply(A=@n10, B=@n11))) {samplerType=SAMPLERTYPE_Normal, texture=/Game/Characters/Player/Male/Large/Bodies/M_LRG_Clash_V/Textures/Simple_Perlin_N}
  n13 = VectorParameter() {group=21.NoiseB, parameterName=NoiseB_Strength, defaultValue=<rgba 0.02 0.01 0 0>}
  n14 = Multiply(A=AppendVector(A=@n12.R, B=@n12.R), B=AppendVector(A=@n13.R, B=@n13.R))
  n15 = MaterialFunctionCall() {materialFunction=&VectorToRadialValue}
  n16 = VectorParameter() {group=22.NoiseC, parameterName=NoiseC_ScaleAndOffset, defaultValue=<rgba 1 2 0.1 -0.1>}
  n17 = AppendVector(A=@n16.R, B=@n16.R)
  n18 = AppendVector(A=@n16.B, B=@n16.B)
  n19 = TextureCoordinate()
  n20 = TextureSample!(
   UVs=Add(A=Multiply(A=@n15.Radial Coordinates, B=@n17), B=Multiply(A=@n18, B=@n11))
   Tex
   "DDX(UVs)"=DDX(Value=@n19)
   "DDY(UVs)"=DDY(Value=@n19)
   Apply View MipBias
  ) {
   mipValueMode=TMVM_Derivative
   samplerType=SAMPLERTYPE_LinearColor
   texture=/Game/Athena/Apollo/Water/Textures/02_Rounded/03_WaterFoam_Veroni_base_H
  }
  n21 = VectorParameter() {group=22.NoiseC, parameterName=NoiseC_StrengthAndOffset, defaultValue=<rgba 0.01 0.01 0 0>}
  n22 = Multiply(A=AppendVector(A=@n20.R, B=@n20.R), B=AppendVector(A=@n21.R, B=@n21.R))
  n23 = AppendVector(A=@n21.B, B=@n21.B)
  n24 = Add(A=ShadingPathSwitch(Default=@n14, Deferred=@n14, Forward=@n14, Mobile=Constant()), B=Add(A=Add(A=@n22, B=@n23), B=@n15.Radial Coordinates))
  MaterialFunctionCall(
   Normal=MaterialFunctionCall!(
    Texture=@n4
    "[Override]DDXInput"=DDX(Value=@n5)
    "[Override]DDYInput"=DDY(Value=@n5)
    AdditionalUVOffset
    UVControl_RGB=@n6
    UVControl_A=@n6
    "[Override]TimeInput"
    "[Override]UVInput"=@n24
   ) {materialFunction=/Game/Effects/Niagara/MasterMaterials/Functions/Common/FXCommon_TextureSample_DDXDDY}.RGB
   Flatness=ScalarParameter() {defaultValue=0.5, parameterName=NormalFlat}
  ) {materialFunction=&FlattenNormal}
  MaterialFunctionCall() {materialFunction=&DitherTemporalAA_Fortnite}
  n25 = MaterialFunctionCall(Float2=@n15.Radial Coordinates) {materialFunction=&BreakOutFloat2Components}
  n26 = OneMinus(None=Saturate(None=@n25.G))
  n27 = ScalarParameter() {defaultValue=2, parameterName=awdwad}
  n28 = LinearInterpolate(
   A=ScalarParameter() {defaultValue=2, parameterName=C}
   B=ScalarParameter() {parameterName=c2}
   Alpha=OneMinus(None=Power(Base=@n26, Exp=@n27))
  )
  Multiply(
   A=LinearInterpolate(
    A=VectorParameter() {parameterName=EdgeC, defaultValue=<rgba 0 0.52 0.6 1>}.RGB
    B=VectorParameter() {parameterName=CentC, defaultValue=<rgba 0.84 1.12 1.4 1>}.RGB
    Alpha=@n28
   )
   B=Normalize(VectorInput=@n3.RGB)
  )
  n29 = MaterialFunctionCall(Float2=@n15.Radial Coordinates) {materialFunction=&BreakOutFloat2Components}
  n30 = Multiply(A=OneMinus(None=@n29.R)) {constB=5}
  n31 = Multiply(A=@n29.R) {constB=5}
  n32 = ScalarParameter() {defaultValue=8, parameterName=CenterFade}
  n33 = Power(Base=Saturate(None=Multiply(A=@n25.G, B=@n32))) {constExponent=5}
  n34 = Constant() {r=1}
  n35 = ScreenPosition()
  n36 = MaterialFunctionCall!(OffsetAmount=ScalarParameter() {defaultValue=0.03, parameterName=Scale}) {materialFunction=/Engine/Functions/Engine_MaterialFunctions01/ImageAdjustment/BlurSampleOffsets}
  MaterialExpressionScreenPosition_1 = ScreenPosition()
  n37 = Custom!(ScreenPos=@MaterialExpressionScreenPosition_1.PixelPosition) {
   code="float random = InterleavedGradientNoise(ScreenPos.xy, float(View.StateFrameIndexMod8));\r\nreturn random;"
   outputType=CMOT_Float1
   inputs [{inputName=ScreenPos, input {expression=&self:MaterialExpressionScreenPosition_1, outputIndex=1}}]
  }
  n38 = Divide(A=@n37, B=ScalarParameter() {defaultValue=2, parameterName=DivideDither}) {constB=1.5}
  n39 = SceneTexture(UVs=Add(A=@n35.ViewportUV, B=Multiply(A=@n36."0,1", B=@n38))) {sceneTextureId=PPI_ShadingModelID}
  n40 = Constant()
  n41 = Constant() {r=1}
  n42 = SceneTexture(UVs=Add(A=@n35.ViewportUV, B=Multiply(A=@n36."1,0", B=@n38))) {sceneTextureId=PPI_ShadingModelID}
  n43 = Constant()
  n44 = Constant() {r=1}
  n45 = Add(A=@n35.ViewportUV, B=Multiply(A=@n36."0,-1", B=@n38))
  n46 = Constant()
  n47 = Add(
   A=Add(
    A=If(A=@n39.Color, "A > B"=@n40, "A == B"=@n41, "A < B"=@n40) {constB=10, equalsThreshold=0.01}
    B=If(A=@n42.Color, "A > B"=@n43, "A == B"=@n44, "A < B"=@n43) {constB=10, equalsThreshold=0.01}
   )
   B=If(A=SceneTexture(UVs=@n45) {sceneTextureId=PPI_ShadingModelID}.Color, "A > B"=@n46, "A == B"=Constant() {r=1}, "A < B"=@n46) {constB=10, equalsThreshold=0.01}
  )
  n48 = Multiply(A=@n36."-1,0", B=@n38)
  n49 = Constant()
  n50 = If(
   A=SceneTexture(UVs=Add(A=@n35.ViewportUV, B=@n48)) {sceneTextureId=PPI_ShadingModelID}.Color
   "A > B"=@n49
   "A == B"=Constant() {r=1}
   "A < B"=@n49
  ) {constB=10, equalsThreshold=0.01}
  n51 = SceneTexture(UVs=Add(A=@n35.ViewportUV, B=Multiply(A=@n36."-.5,.5", B=@n38))) {sceneTextureId=PPI_ShadingModelID}
  n52 = Constant()
  n53 = Constant() {r=1}
  n54 = Add(A=@n35.ViewportUV, B=Multiply(A=@n36.".5,.5", B=@n38))
  n55 = Constant()
  n56 = Add(
   A=Add(A=Add(A=@n47, B=@n50), B=If(A=@n51.Color, "A > B"=@n52, "A == B"=@n53, "A < B"=@n52) {constB=10, equalsThreshold=0.01})
   B=If(A=SceneTexture(UVs=@n54) {sceneTextureId=PPI_ShadingModelID}.Color, "A > B"=@n55, "A == B"=Constant() {r=1}, "A < B"=@n55) {constB=10, equalsThreshold=0.01}
  )
  n57 = Multiply(A=@n36.".5,-.5", B=@n38)
  n58 = Constant()
  n59 = If(
   A=SceneTexture(UVs=Add(A=@n35.ViewportUV, B=@n57)) {sceneTextureId=PPI_ShadingModelID}.Color
   "A > B"=@n58
   "A == B"=Constant() {r=1}
   "A < B"=@n58
  ) {constB=10, equalsThreshold=0.01}
  n60 = SceneTexture(UVs=Add(A=@n35.ViewportUV, B=Multiply(A=@n36."-.5,-.5", B=@n38))) {sceneTextureId=PPI_ShadingModelID}
  n61 = Constant()
  n62 = Constant() {r=1}
  n63 = Divide(A=Add(A=Add(A=@n56, B=@n59), B=If(A=@n60.Color, "A > B"=@n61, "A == B"=@n62, "A < B"=@n61) {constB=10, equalsThreshold=0.01})) {constB=8}
  n64 = TextureObjectParameter() {
   group=RippleTexture
   parameterName=RippleTexture_Texture
   samplerType=SAMPLERTYPE_LinearColor
   sortPriority=0
   texture=/Game/Global/Textures/Default/Caustics/03_WaterFoam_AccentSwirls_H
  }
  n65 = TextureCoordinate()
  n66 = MaterialFunctionCall(
   Texture=@n64
   "[Override]DDXInput"=DDX(Value=@n65)
   "[Override]DDYInput"=DDY(Value=@n65)
   UVControl_RGB=@n6
   UVControl_A=@n6
   "[Override]UVInput"=@n24
  ) {materialFunction=&FXCommon_TextureSample_DDXDDY}
  n67 = ScalarParameter() {defaultValue=0.5, parameterName=ad}
  n68 = Power(Base=@n25.G) {constExponent=0.95}
  n69 = ScalarParameter() {defaultValue=1.8}
  n70 = Multiply(A=Subtract(A=Add(A=@n66.R, B=@n67), B=Multiply(A=@n68, B=@n69)), B=ScalarParameter() {defaultValue=1, parameterName=M})
  n71 = Multiply(
   A=LinearInterpolate(
    B=ShadingPathSwitch(Default=@n63, Deferred=@n63, Forward=@n34, Mobile=@n34)
    Alpha=ScalarParameter() {defaultValue=1, parameterName=Use Shader Id Mask}
   ) {constA=1}
   B=MaterialFunctionCall(In=Saturate(None=@n70), Contrast=@n28) {materialFunction=&CheapContrast}.Result
  )
  Multiply(A=Multiply(A=Saturate(None=@n30), B=Saturate(None=@n31)), B=Multiply(A=@n3.A, B=Multiply(A=@n33, B=@n71)))
 ]
}
