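# Translucent, volumetric-lit Niagara sprite material for waterfall smoke, with responsive AA.
# Two panning Perlin noise samples, time-offset by a per-particle random dynamic parameter, distort the smoke shape texture's UVs.
# Opacity combines that shape, particle alpha and edge masks repeated on a 2x2 UV grid, softened by depth and camera fades.
# Color is tinted by normalized particle color, and a rainbow function plus particle color feeds emissive adjustments.
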Material /FigmentEnvironment/Water/M_Waterfall_Smoke
graph {
 properties {
  bEnableResponsiveAA=true
  bUsedWithNiagaraSprites=true
  blendMode=BLEND_Translucent
  pixelDepthOffsetMode=PDOM_Legacy
  translucencyLightingMode=TLM_VolumetricPerVertexNonDirectional
  usedShadingModels=Default Lit
 }
 nodes [
  n1 = TextureObjectParameter() {
   group=NoiseA
   parameterName=NoiseA_Texture
   samplerType=SAMPLERTYPE_Normal
   sortPriority=0
   texture=/Game/Characters/Player/Male/Large/Bodies/M_LRG_Clash_V/Textures/Simple_Perlin_N
  }
  n2 = VectorParameter() {group=NoiseA, parameterName=NoiseA_UVControl, channelNames {a=VPan, b=UPan, g=VScale, r=UScale}, defaultValue=<rgba 0.5 0.5 0 0.2>}
  n3 = Add(
   A=Time() {bOverride_Period=true, period=3000}
   B=Multiply(A=DynamicParameter() {paramNames [Random,Param2,Param3,Param4]}.Random) {constB=100}
  )
  n4 = VectorParameter() {
   group=NoiseA
   parameterName=NoiseA_NoiseControl
   sortPriority=60
   channelNames {a=VStrength, b=UStrength, g=VOffset, r=UOffset}
   defaultValue=<rgba -0.5 0 0 0.2>
  }
  n5 = MaterialFunctionCall!(
   NoiseValue=MaterialFunctionCall(Texture=@n1, UVControl_RGB=@n2, UVControl_A=@n2, "[Override]TimeInput"=@n3) {materialFunction=&FXCommon_TextureSample}.RG
   Control_RGB=@n4
   Control_A=@n4
  ) {materialFunction=/Game/Effects/Niagara/MasterMaterials/Functions/Common/FXCommon_NoiseControl}
  Add(A=TextureCoordinate(), B=@n5.NoiseValue)
  Constant() {r=1}
  n6 = ParticleColor()
  MaterialFunctionCall(Emissive=Add(A=MaterialFunctionCall() {materialFunction=/Game/Tools/WaterfallGenerator/Art/MF_Rainbow}.Result, B=@n6.RGB)) {materialFunction=&MF_FX_EmissiveAdjustments}
  n7 = MaterialFunctionCall(Float2=Frac(None=TextureCoordinate() {uTiling=2, vTiling=2})) {materialFunction=&BreakOutFloat2Components}
  n8 = VectorParameter() {group=EdgeMask, parameterName=EdgeMask_EdgeMask_Control, channelNames {a="", b=Power, g=Multiply, r=Offset}, defaultValue=<rgba 0 2 1.2 1>}
  n9 = Multiply(
   A=MaterialFunctionCall!(UorV=@n7.R, EdgeControl=@n8) {materialFunction=/Game/Effects/Niagara/MasterMaterials/Functions/Common/FXCommon_EdgeMask_Simple}.Mask
   B=MaterialFunctionCall(UorV=@n7.G, EdgeControl=@n8) {materialFunction=&FXCommon_EdgeMask_Simple}.Mask
  )
  n10 = TextureObjectParameter() {
   group=NoiseB
   parameterName=NoiseB_Texture
   samplerType=SAMPLERTYPE_Normal
   sortPriority=0
   texture=&Simple_Perlin_N
  }
  n11 = VectorParameter() {group=NoiseB, parameterName=NoiseB_UVControl, channelNames {a=VPan, b=UPan, g=VScale, r=UScale}, defaultValue=<rgba 2 2 0 0.1>}
  n12 = MaterialFunctionCall(Texture=@n10, UVControl_RGB=@n11, UVControl_A=@n11, "[Override]TimeInput"=@n3) {materialFunction=&FXCommon_TextureSample}
  n13 = VectorParameter() {
   group=NoiseB
   parameterName=NoiseB_NoiseControl
   sortPriority=60
   channelNames {a=VStrength, b=UStrength, g=VOffset, r=UOffset}
   defaultValue=<rgba 0 0 0 0.03>
  }
  n14 = Add(
   A=@n5.NoiseValue
   B=Add(
    A=MaterialFunctionCall(NoiseValue=@n12.RG, Control_RGB=@n13, Control_A=@n13) {materialFunction=&FXCommon_NoiseControl}.NoiseValue
    B=TextureCoordinate()
   )
  )
  n15 = TextureObjectParameter() {
   group=SmokeTexture
   parameterName=SmokeTexture_Texture
   samplerType=SAMPLERTYPE_LinearGrayscale
   sortPriority=0
   texture=/Game/Tools/WaterfallGenerator/Art/T_Waterfall_SmokeShape
  }
  n16 = VectorParameter() {group=SmokeTexture, parameterName=SmokeTexture_UVControl, channelNames {a=VPan, b=UPan, g=VScale, r=UScale}, defaultValue=<rgba 1 1 0 0>}
  n17 = Multiply(
   A=MaterialFunctionCall!(TexCoord=@n14, ClampChannel=Constant2Vector() {g=1, r=1}) {materialFunction=/Game/Effects/Fort_Effects/MaterialFunctions/MF_FX_TilingMethod}.Result
   B=Multiply(
    A=MaterialFunctionCall(Texture=@n15, UVControl_RGB=@n16, UVControl_A=@n16, "[Override]UVInput"=@n14) {materialFunction=&FXCommon_TextureSample}.R
    B=@n6.A
   )
  )
  Multiply(
   A=DepthFade(Opacity=Multiply(A=@n9, B=@n17), FadeDistance=ScalarParameter() {defaultValue=100, parameterName=DF})
   B=MaterialFunctionCall(
    Fade Length=ScalarParameter() {defaultValue=600, parameterName="Fade Length (S)"}
    Fade Offset=ScalarParameter() {defaultValue=400, parameterName="Fade Offset (S)"}
   ) {materialFunction=&CameraDepthFade}.Result
  )
  Multiply(A=VectorParameter() {parameterName=Color, defaultValue=<rgba 1 1 1 1>}.RGB, B=Normalize(VectorInput=@n6.RGB))
 ]
}
