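# Translucent, lit Niagara sprite material for waterfall splash particles.
# Dynamic parameters scale two noise UV distortions and a vertical stretch around quadrant centers of a 2x2-split splash texture.
# The splash channels blend toward a noise-warped smoke shape by another dynamic parameter.
# The result is multiplied by particle alpha and a clamped per-quadrant mask, then depth-faded, except on the mobile path.
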Material /FigmentEnvironment/Water/M_Waterfall_Splash
graph {
 properties {bUsedWithNiagaraSprites=true, blendMode=BLEND_Translucent, pixelDepthOffsetMode=PDOM_Legacy, usedShadingModels=Default Lit}
 nodes [
  n1 = TextureObjectParameter() {
   group=MainTexture
   parameterName=MainTexture_Texture
   samplerType=SAMPLERTYPE_LinearColor
   sortPriority=0
   texture=/Game/Effects/Fort_Effects/Textures/Water/T_waterSplash
  }
  n2 = VectorParameter() {group=MainTexture, parameterName=MainTexture_UVControl, channelNames {a=VPan, b=UPan, g=VScale, r=UScale}, defaultValue=<rgba 1 1 0 0>}
  n3 = TextureCoordinate()
  n4 = Multiply(A=Floor(None=Multiply(A=@n3) {constB=1.9999})) {constB=0.5}
  n5 = Add(A=@n4) {constB=0.25}
  n6 = TextureObjectParameter() {
   group=NoiseB
   parameterName=NoiseB_Texture
   sortPriority=0
   texture=/Engine/Functions/Engine_MaterialFunctions02/ExampleContent/Textures/smoothNormalSphere
  }
  n7 = VectorParameter() {group=NoiseB, parameterName=NoiseB_UVControl, channelNames {a=VPan, b=UPan, g=VScale, r=UScale}, defaultValue=<rgba 2 2 0 0>}
  n8 = Multiply(
   A=MaterialFunctionCall(Texture=@n6, UVControl_RGB=@n7, UVControl_A=@n7) {materialFunction=&FXCommon_TextureSample}.RG
   B=DynamicParameter() {defaultValue=<rgba 1 0 0 0>}.Param2
  )
  n9 = TextureObjectParameter() {
   group=NoiseA
   parameterName=NoiseA_Texture
   samplerType=SAMPLERTYPE_LinearGrayscale
   sortPriority=0
   texture=&T_SimplePerlin_Greyscale
  }
  n10 = VectorParameter() {group=NoiseA, parameterName=NoiseA_UVControl, channelNames {a=VPan, b=UPan, g=VScale, r=UScale}, defaultValue=<rgba 1 1 0 0.1>}
  n11 = MaterialFunctionCall(Texture=@n9, UVControl_RGB=@n10, UVControl_A=@n10) {materialFunction=&FXCommon_TextureSample}
  n12 = MaterialFunctionCall(R=@n11.R, G=@n11.R) {materialFunction=&AppendMany}
  n13 = VectorParameter() {
   group=NoiseAToMain
   parameterName=NoiseAToMain_NoiseControl
   sortPriority=60
   channelNames {a=VStrength, b=UStrength, g=VOffset, r=UOffset}
   defaultValue=<rgba -0.5 -0.5 0.01 0.05>
  }
  n14 = MaterialFunctionCall!(NoiseValue=@n12.RG, Control_RGB=@n13, Control_A=@n13) {materialFunction=/Game/Effects/Niagara/MasterMaterials/Functions/Common/FXCommon_NoiseControl}
  n15 = DynamicParameter() {defaultValue=<rgba 1 0 0 0>}
  n16 = TextureCoordinate()
  n17 = Add(A=@n8, B=Add(A=LinearInterpolate(B=@n14.NoiseValue, Alpha=@n15.Param3), B=@n16))
  n18 = Subtract(A=@n17, B=@n5) {constB=0.5}
  n19 = AppendVector(A=Constant() {r=1}, B=ScalarParameter() {defaultValue=3, parameterName=Test2222})
  n20 = Frac(None=Multiply(A=TextureCoordinate()) {constB=2})
  n21 = Sine(None=Multiply(A=MaterialFunctionCall(Float2=@n20) {materialFunction=&BreakOutFloat2Components}.R) {constB=0.5})
  n22 = LinearInterpolate(
   A=@n17
   B=Add(A=Multiply(A=@n18, B=@n19) {constB=0.9}, B=@n5) {constB=0.5}
   Alpha=LinearInterpolate(B=Power(Base=@n21) {constExponent=30}, Alpha=DynamicParameter() {defaultValue=<rgba 1 0 0 0>}.Param4)
  )
  n23 = MaterialFunctionCall(Texture=@n1, UVControl_RGB=@n2, UVControl_A=@n2, "[Override]UVInput"=@n22) {materialFunction=&FXCommon_TextureSample}
  n24 = Power(Base=@n23.R)
  n25 = Power(Base=@n23.G)
  Add(A=@n24, B=@n25)
  n26 = VectorParameter() {
   group=NoiseA
   parameterName=NoiseA_NoiseControl
   sortPriority=60
   channelNames {a=VStrength, b=UStrength, g=VOffset, r=UOffset}
   defaultValue=<rgba -0.5 -0.5 0.3 0.3>
  }
  n27 = Add(
   A=MaterialFunctionCall(NoiseValue=@n12.RG, Control_RGB=@n26, Control_A=@n26) {materialFunction=&FXCommon_NoiseControl}.NoiseValue
   B=@n16
  )
  Add(A=@n27, B=@n8)
  n28 = ParticleColor()
  MaterialFunctionCall(Emissive=@n28.RGB) {materialFunction=&MF_FX_EmissiveAdjustments}
  Normalize(VectorInput=@n28.RGB)
  n29 = Multiply(A=Subtract(A=@n22, B=@n4)) {constB=2}
  n30 = Constant() {r=1}
  n31 = TextureObjectParameter() {
   group=GlowTexture
   parameterName=GlowTexture_Texture
   samplerType=SAMPLERTYPE_LinearGrayscale
   sortPriority=0
   texture=/Game/Tools/WaterfallGenerator/Art/T_Waterfall_SmokeShape
  }
  n32 = VectorParameter() {group=GlowTexture, parameterName=GlowTexture_UVControl, channelNames {a=VPan, b=UPan, g=VScale, r=UScale}, defaultValue=<rgba 1 1 0 0>}
  n33 = MaterialFunctionCall(Texture=@n31, UVControl_RGB=@n32, UVControl_A=@n32, "[Override]UVInput"=@n27) {materialFunction=&FXCommon_TextureSample}
  n34 = LinearInterpolate(
   A=LinearInterpolate(A=@n24, B=@n25)
   B=LinearInterpolate(A=@n33.R, B=Power(Base=@n33.R) {constExponent=9}) {constAlpha=0.8}
   Alpha=DynamicParameter() {parameterIndex=1, defaultValue=<rgba 1 0 0 0>}.Param1
  ) {constAlpha=0.15}
  n35 = Multiply(A=Multiply(
    A=MaterialFunctionCall!(TexCoord=@n29, ClampChannel=@n30) {materialFunction=/Game/Effects/Fort_Effects/MaterialFunctions/MF_FX_TilingMethod}.Result
    B=Multiply(A=@n28.A, B=@n34)
   )) {constB=1.5}
  n36 = DepthFade(Opacity=@n35, FadeDistance=ScalarParameter() {parameterName=FadeDistance})
  ShadingPathSwitch(Default=@n36, Deferred=@n36, Mobile=@n35)
 ]
}
