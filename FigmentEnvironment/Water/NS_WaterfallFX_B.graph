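# Niagara waterfall effect that layers sprite emitters for edge flow, falling drops and mist at the top, and ripples, smoke and splashes at the bottom.
# User top and bottom positions place the layers, while fall width spreads particles sideways and scales most spawn rates.
# A Fountain emitter spawns hit splashes from a user position array via spawn groups.
# System-level time-of-day color and alpha tint particles, and many emitters use distance culling.
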NiagaraSystem /FigmentEnvironment/Water/NS_WaterfallFX_B
dependencies {
 usedDataInterfaces [&NiagaraDataInterfaceCurve, &NiagaraDataInterfaceVectorCurve, &NiagaraDataInterfaceCamera]
 usedDynamicInputs [
  &Subtract_Float
  &Abs_Float
  &DotProduct
  &ReturnCameraVector
  &FN_Select_Float_from_Time_Of_Day
  &FN_Select_Color_from_Time_Of_Day
 ]
 usedModules [
  &SystemState
  &self:SystemSpawnScript.SystemScriptSource.SystemScriptGraph.NiagaraNodeAssignment_1.SetVariables_#1
  &self:SystemSpawnScript.SystemScriptSource.SystemScriptGraph.NiagaraNodeAssignment_0.SetVariables_#2
  &self:SystemSpawnScript.SystemScriptSource.SystemScriptGraph.NiagaraNodeAssignment_2.SetVariables_#3
  &self:SystemSpawnScript.SystemScriptSource.SystemScriptGraph.NiagaraNodeAssignment_4.SetVariables_#4
 ]
}
emitters {
 BottomRippleA {
  emitterSpawnScript [
   /SetVariables(Emitter.UseIdMask:NiagaraBool=$User.UseIdMask) {isSetParameters=true}
   /SetVariables(Emitter.DontUseIDmask:NiagaraBool=&InvertBool(Boolean:NiagaraBool=$User.UseIdMask)) {isSetParameters=true}
  ]
  emitterUpdateScript [
   &EmitterState(
    Inactive Response=::NewEnumerator1
    Life Cycle Mode=::NewEnumerator1
    Loop Behavior=::NewEnumerator1
    Loop Duration Mode=::NewEnumerator1
    UseLoopDelay=false
   )
   &SpawnBurst_Instantaneous(Use Spawn Probability=false, UseLoopCountLimit=false)
  ]
  particleSpawnScript [&InitializeParticle(
    Color=&MultiplyLinearColor(
     Linear Color=$User.BottomSplash Color
     Scale Factor=&MakeVector4FromVectorAndFloat("Vector (RGB)"=<1 1 1>, "Float (Alpha)"=$User.RippleOpacity)
    )
    Lifetime=5
    Position=&TransformPosition(Position=$User.BottomPosition, Source Space=::World, Destination Space=::Local)
    Position Mode=::NewEnumerator1
    Position Offset=&MakeVector(X=0, Y=0, Z=$User.RippleOffset)
    Position Offset Coordinate Space=::Local
    Sprite Rotation Angle=90
    Sprite Rotation Mode=::NewEnumerator0
    Sprite Size Mode=::NewEnumerator3
    Uniform Sprite Size=$User.RippleWidth
    UsePositionOffset=true
    Write Material Random=true
   )]
  particleUpdateScript [
   &ParticleState(Kill Particles When Lifetime Has Elapsed=false, Let Infinitely Lived Particles Die When Emitter Deactivates=true)
   &SpriteFacingAndAlignment(Sprite Facing=<0 0 1>, WriteAlignment=true, WriteFacing=true)
   &ScaleColor(
    Scale Alpha=&FNTimeOfDayScalar(Night Scale=1.25, Morning Scale=0.5, Evening Scale=0.5)
    Scale RGB=&MakeVectorFromLinearColorRGB(Color=&FNTimeOfDayColor(Night Color=<rgba 0.6 0.8 1 1>, Morning Color=<rgba 0.4 0.25 0.22 1>, Day Color=<rgba 1 1 1 1>, Evening Color=<rgba 0.25 0.125 0.1 1>))
    ScaleA=true
    ScaleRGB=true
   ) {enabled=false, module=ScaleColor001}
   &ScaleColor(
    Scale Alpha=&self:Sqr(Origin:NiagaraFloat=$System.TOD_Alpha)
    Scale RGB=&MakeVectorFromLinearColorRGB(Color=$System.TOD_Color)
    ScaleA=true
    ScaleRGB=true
   ) {module=ScaleColor002}
  ]
  rendererClasses [&NiagaraSpriteRendererProperties]
  renderers [
   {rendererClass=&NiagaraSpriteRendererProperties}
   {rendererClass=&NiagaraSpriteRendererProperties, rendererIndex=1}
   {rendererClass=&NiagaraSpriteRendererProperties, rendererIndex=2}
  ]
 }
 BottomSmokeA {
  emitterUpdateScript [
   &EmitterState(
    Enable Distance Culling=true
    MaxDistance=15000
    Module.bUseMaxDistance=true
    Module.bUseMinDistance=false
    Scalability Mode=::NewEnumerator1
   )
   &SpawnRate(
    SpawnRate=&Multiply_Float(
     A=&Multiply_Float(A=10, B=$User.BottomSmokeDensity)
     B=&Divide_Float(A=$User.FallWidth, B=850)
    )
    Use Spawn Probability=false
   )
  ]
  particleSpawnScript [
   &InitializeParticle(
    Color=<rgba 1 1 1 0.55>
    Lifetime Max=7
    Lifetime Min=6
    Lifetime Mode=::NewEnumerator1
    Sprite Size Mode=::NewEnumerator3
    Uniform Sprite Size=&Lerp_Float(
     A=600
     B=800
     Alpha=&FloatFromCurve(
      FloatCurve=DataInterface() {propertyValues {Curve {keys [
          {interpMode=RCIM_Cubic, leaveTangent=0.1305, tangentMode=RCTM_Break, time=0, value=0}
          {arriveTangent=1, interpMode=RCIM_Cubic, leaveTangent=1, time=0.7, value=0.3}
          {arriveTangent=2.0211, interpMode=RCIM_Cubic, tangentMode=RCTM_Break, time=1, value=1}
         ]}}}
      CurveIndex=&RandomRangeFloat(Minimum=0, Maximum=1)
     )
    )
    UsePositionOffset=false
    Write Material Random=true
   )
   &SubUVAnimation(End Frame=3, SubUV Animation Mode:ENiagara_SubUVLookupMode=::NewEnumerator2)
   &ShapeLocation(
    MaskSpawnGroup=false
    Override Seed=false
    Shape Origin=$User.BottomPosition
    Sphere Radius=25
    Sphere Surface Distribution=1
    UseOwnerScale=true
   )
   &AddVelocity(
    Module.Use Falloff Distance=false
    Override Seed=false
    Velocity Mode=::NewEnumerator1
    Velocity Origin=$User.BottomPosition
    Velocity Speed=600
   ) {module=AddVelocity001}
   /SetVariables(Particles.Velocity=&Multiply_Vector(A=$Particles.Velocity, B=<1 1 0.2>)) {isSetParameters=true, module=SetVariables_#5}
   &AddVelocity(Velocity=<0 0 -50>)
   /SetVariables(Particles.Position=&AddVectorToPosition(
     Position=$Particles.Position
     Vector=&Multiply_VectorByFloat(
      Vector=&TransformVector(Vector=<1 0 0>, Destination Space=::Simulation)
      Float=&Multiply_Float(
       A=&Lerp_Float(
        A=0
        B=&Divide_Float(A=$User.FallWidth, B=2)
        Alpha=&FloatFromCurve(
         FloatCurve=DataInterface() {propertyValues {Curve {keys [
             {interpMode=RCIM_Cubic, leaveTangent=3.1287, tangentMode=RCTM_Break, time=0, value=0}
             {arriveTangent=-0.0122, interpMode=RCIM_Cubic, tangentMode=RCTM_Break, time=1, value=1}
            ]}}}
         CurveIndex=&RandomRangeFloat(Minimum=0, Maximum=1)
        )
       )
       B=&Lerp_Float(
        A=-1
        B=1
        Alpha=&MakeFloatFromInt(Int=&RandomRangeInt(Minimum=0, Maximum=1))
       )
      )
     )
    )) {isSetParameters=true}
  ]
  particleUpdateScript [
   &ParticleState()
   &SpriteRotationRate(Rotation Rate=-10)
   &ScaleColor(
    Scale Alpha=&FloatFromCurve(FloatCurve=DataInterface() {propertyValues {Curve {keys [
         {interpMode=RCIM_Cubic, tangentMode=RCTM_User, time=0, value=0}
         {interpMode=RCIM_Cubic, tangentMode=RCTM_User, time=0.5, value=1}
         {tangentMode=RCTM_User, time=1, value=0}
        ]}}})
    ScaleA=true
    ScaleRGB=true
   )
   &Drag(Drag=0.1636, Use Linear Drag=true, Use Rotational Drag=true)
   &PointAttractionForce(
    Attraction Radius=10000
    Attraction Strength=&FloatFromCurve(
     FloatCurve=DataInterface() {propertyValues {Curve {keys [{interpMode=RCIM_Cubic, tangentMode=RCTM_User, time=0, value=0}, {interpMode=RCIM_Cubic, tangentMode=RCTM_User, time=1, value=1}]}}}
     Scale Curve=20
    )
    Attractor Position:NiagaraPosition=$User.BottomPosition
    Attractor Position Offset:Vector3f=<0 0 200>
    Kill Within Radius=false
    Module.Use Falloff=true
    UseAttractorOffset=true
   )
   &GravityForce(Gravity=&VectorFromCurve(
     VectorCurve=DataInterface() {propertyValues {XCurve {keys [<at 0 0>,<at 1 0>]}, YCurve {keys [<at 0 0>,<at 1 0>]}, ZCurve {keys [<at 0 0>,<at 0.5 0.5>,<at 0.7 -1>]}}}
     Scale Curve=<0 0 300>
    ))
   &SolveForcesAndVelocity(Clamp Velocity=false, Limit Acceleration=false)
   &DynamicMaterialParameters(
    Index 0 Param 1=&RandomRangeFloat(Minimum=0, Maximum=1)
    Module.Param0WriteEnabled=true
    Module.Param1WriteEnabled=true
    Module.Param2WriteEnabled=true
    Module.Param3WriteEnabled=true
   )
   &ScaleColor(
    Scale Alpha=&FNTimeOfDayScalar(Night Scale=0.7, Morning Scale=0.3, Evening Scale=0.3)
    Scale RGB=&MakeVectorFromLinearColorRGB(Color=&FNTimeOfDayColor(Night Color=<rgba 0.6 0.8 1 1>, Morning Color=<rgba 1 0.625 0.55 1>, Day Color=<rgba 1 1 1 1>, Evening Color=<rgba 1 0.5 0.4 1>))
    ScaleA=true
    ScaleRGB=true
   ) {enabled=false, module=ScaleColor001}
   &ScaleColor(
    Scale Alpha=$System.TOD_Alpha
    Scale RGB=&MakeVectorFromLinearColorRGB(Color=$System.TOD_Color)
    ScaleA=true
    ScaleRGB=true
   ) {module=ScaleColor002}
   &KillParticles(Kill Particles=&SetBoolByFloatComparison(A=$Particles.NormalizedAge, B=0.85))
  ]
  rendererClasses [&NiagaraSpriteRendererProperties]
  renderers [{rendererClass=&NiagaraSpriteRendererProperties}]
 }
 BottomSmokeB {
  emitterUpdateScript [
   &EmitterState(
    Enable Distance Culling=true
    MaxDistance=20000
    Module.bUseMaxDistance=true
    Module.bUseMinDistance=false
    Scalability Mode=::NewEnumerator1
   )
   &SpawnRate(
    SpawnRate=&Multiply_Float(
     A=20
     B=&Multiply_Float(A=$User.BottomSmokeDensity, B=&Divide_Float(A=$User.FallWidth, B=850))
    )
    Use Spawn Probability=false
   )
  ]
  particleSpawnScript [
   &InitializeParticle(
    Color=<rgba 1 1 1 0.3>
    Lifetime=5
    Sprite Rotation Mode=::NewEnumerator2
    Sprite Size Mode=::NewEnumerator1
    Uniform Sprite Size Max=800
    Uniform Sprite Size Min=600
    UsePositionOffset=false
    Write Material Random=true
   )
   &ShapeLocation(
    MaskSpawnGroup=false
    Override Seed=false
    Shape Origin=$User.BottomPosition
    Sphere Radius=25
    Sphere Surface Distribution=1
    UseOwnerScale=true
   )
   &AddVelocity(
    Module.Use Falloff Distance=false
    Override Seed=false
    Velocity Mode=::NewEnumerator1
    Velocity Origin=$User.BottomPosition
    Velocity Speed=&RandomRangeFloat(Minimum=750.4585, Maximum=1060.5421)
   )
   /SetVariables(Particles.Velocity=&Multiply_Vector(A=$Particles.Velocity, B=<1 1 0>)) {isSetParameters=true}
   &AddVelocity(Velocity=<0 0 -50>) {module=AddVelocity001}
   /SetVariables(Particles.Position=&AddVectorToPosition(
     Position=$Particles.Position
     Vector=&Multiply_VectorByFloat(
      Vector=&TransformVector(Vector=<1 0 0>, Destination Space=::Simulation)
      Float=&Multiply_Float(
       A=&Lerp_Float(
        A=0
        B=&Divide_Float(A=$User.FallWidth, B=2)
        Alpha=&FloatFromCurve(
         FloatCurve=DataInterface() {propertyValues {Curve {keys [
             {interpMode=RCIM_Cubic, leaveTangent=3.1287, tangentMode=RCTM_Break, time=0, value=0}
             {arriveTangent=-0.0122, interpMode=RCIM_Cubic, tangentMode=RCTM_Break, time=1, value=1}
            ]}}}
         CurveIndex=&RandomRangeFloat(Minimum=0, Maximum=1)
        )
       )
       B=&Lerp_Float(
        A=-1
        B=1
        Alpha=&MakeFloatFromInt(Int=&RandomRangeInt(Minimum=0, Maximum=1))
       )
      )
     )
    )) {isSetParameters=true}
  ]
  particleUpdateScript [
   &ParticleState()
   &ScaleColor(
    Scale Alpha=&FloatFromCurve(FloatCurve=DataInterface() {propertyValues {Curve {keys [
         {interpMode=RCIM_Cubic, tangentMode=RCTM_User, time=0, value=0}
         {interpMode=RCIM_Cubic, tangentMode=RCTM_User, time=0.5, value=1}
         {tangentMode=RCTM_User, time=1, value=0}
        ]}}})
    ScaleA=true
    ScaleRGB=true
   )
   &GravityForce(Gravity=&VectorFromCurve(
     VectorCurve=DataInterface() {propertyValues {
       XCurve {keys [<at 0 1>,<at 1 0>]}
       YCurve {keys [<at 0 1>,<at 1 0>]}
       ZCurve {keys [{interpMode=RCIM_Cubic, leaveTangent=2.1548, tangentMode=RCTM_Break, time=0, value=0}, <at 1 1>]}
      }}
     Scale Curve=<0 0 150>
    ))
   &Drag(Drag=0.75, Use Linear Drag=true, Use Rotational Drag=true)
   &CurlNoiseForce(Noise Frequency=1, Noise Strength=250, Pan Noise=true, Pan Noise Field=<0 0 0.3>)
   &SolveForcesAndVelocity(Clamp Velocity=false, Limit Acceleration=false)
   &CameraOffset(Camera Offset Amount=300)
   &ScaleColor(
    Scale Alpha=&FNTimeOfDayScalar(Night Scale=0.5, Morning Scale=0.1, Evening Scale=0.2)
    Scale RGB=&MakeVectorFromLinearColorRGB(Color=&FNTimeOfDayColor(Night Color=<rgba 0.6 0.8 1 1>, Morning Color=<rgba 1 0.625 0.55 1>, Day Color=<rgba 1 1 1 1>, Evening Color=<rgba 1 0.5 0.4 1>))
    ScaleA=true
    ScaleRGB=true
   ) {enabled=false, module=ScaleColor001}
   &ScaleColor(
    Scale Alpha=$System.TOD_Alpha
    Scale RGB=&MakeVectorFromLinearColorRGB(Color=$System.TOD_Color)
    ScaleA=true
    ScaleRGB=true
   ) {module=ScaleColor002}
   &KillParticles(Kill Particles=&SetBoolByFloatComparison(A=$Particles.NormalizedAge, B=0.92))
  ]
  rendererClasses [&NiagaraSpriteRendererProperties]
  renderers [{rendererClass=&NiagaraSpriteRendererProperties}]
 }
 BottomSplash {
  emitterUpdateScript [
   &EmitterState(
    Enable Distance Culling=true
    Life Cycle Mode=::NewEnumerator1
    Loop Duration=2
    MaxDistance=10000
    Module.bUseMaxDistance=true
    Module.bUseMinDistance=false
    Scalability Mode=::NewEnumerator1
    UseLoopDelay=false
   )
   &SpawnRate(
    SpawnRate=&Multiply_Float(A=$User.bottomsplashrate, B=&Divide_Float(A=$User.FallWidth, B=850))
    Use Spawn Probability=false
   )
  ]
  particleSpawnScript [
   &InitializeParticle(
    Color Maximum=<rgba 1 1 1 1.5>
    Color Minimum=<rgba 0.7 1 1 1>
    Color Mode=::NewEnumerator2
    Lifetime Max=1.3
    Lifetime Min=0.6
    Lifetime Mode=::NewEnumerator1
    Mass Mode=::NewEnumerator3
    Sprite Rotation Mode=::NewEnumerator2
    Sprite Size Mode=::NewEnumerator1
    Uniform Sprite Size Max=241.489
    Uniform Sprite Size Min=175.759
    UsePositionOffset=false
    Write Material Random=true
   )
   &ShapeLocation(
    Box Size=<0 50 50>
    MaskSpawnGroup=false
    Offset Coordinate Space=::Local
    Offset Mode=::NewEnumerator0
    Override Seed=false
    Shape Origin=$User.BottomPosition
    Shape Primitive=::NewEnumerator2
    Surface Only=false
    UseOwnerScale=true
   )
   /SetVariables(Particles.Position=&AddVectorToPosition(
     Position=$Particles.Position
     Vector=&Multiply_VectorByFloat(
      Vector=&TransformVector(Vector=<1 0 0>, Destination Space=::Simulation)
      Float=&Multiply_Float(
       A=&Lerp_Float(
        A=0
        B=&Divide_Float(A=$User.FallWidth, B=2.2)
        Alpha=&FloatFromCurve(
         FloatCurve=DataInterface() {propertyValues {Curve {keys [
             {interpMode=RCIM_Cubic, leaveTangent=3.3639, tangentMode=RCTM_Break, time=0, value=0}
             {arriveTangent=-0.0122, interpMode=RCIM_Cubic, tangentMode=RCTM_Break, time=1, value=1}
            ]}}}
         CurveIndex=&RandomRangeFloat(Minimum=0, Maximum=1)
        )
       )
       B=&Lerp_Float(
        A=-1
        B=1
        Alpha=&MakeFloatFromInt(Int=&RandomRangeInt(Minimum=0, Maximum=1))
       )
      )
     )
    )) {isSetParameters=true, module=SetVariables_#6}
   &AddVelocity(
    Cone Angle=32
    Cone Axis=<0 0.25 1>
    Distribution Along Cone Axis=0.25
    Module.UseConeFalloff=true
    Module.UseVelDistribution=true
    Override Seed=false
    Speed Falloff From Cone Axis=0.5
    Velocity Mode=::NewEnumerator2
    Velocity Speed=&RandomRangeFloat(Minimum=750, Maximum=1000)
   )
   &SubUVAnimation(End Frame=3)
  ]
  particleUpdateScript [
   &ParticleState()
   &ScaleSpriteSize(
    Uniform Curve Scale=$User.Bottom SPlash Scale
    Uniform Curve Sprite Scale=DataInterface() {propertyValues {Curve {keys [
        {interpMode=RCIM_Cubic, leaveTangent=5.6407, tangentMode=RCTM_Break, time=0, value=0.4}
        {tangentMode=RCTM_User, time=0.3, value=1}
        <at 1 1.5>
       ]}}}
   )
   &GravityForce(Gravity=<0 0 -1200>)
   &ScaleColor(
    Scale Alpha=&FloatFromCurve(
     FloatCurve=DataInterface() {propertyValues {Curve {keys [
         {arriveTangent=13.2056, interpMode=RCIM_Cubic, leaveTangent=13.2056, tangentMode=RCTM_User, time=0, value=0}
         {arriveTangent=-0.4588, interpMode=RCIM_Cubic, leaveTangent=-0.4588, tangentMode=RCTM_User, time=0.2, value=1}
         {arriveTangent=-0.0447, interpMode=RCIM_Cubic, tangentMode=RCTM_Break, time=1, value=0}
        ]}}}
     Scale Curve=0.5
    )
    Scale RGB=<4 4 4>
    ScaleA=true
    ScaleRGB=true
   )
   &ScaleColor(
    Scale Alpha=&FloatFromCurve(FloatCurve=DataInterface() {propertyValues {Curve {keys [
         {arriveTangent=13.2056, interpMode=RCIM_Cubic, leaveTangent=13.2056, tangentMode=RCTM_User, time=0, value=0}
         {arriveTangent=-0.4588, interpMode=RCIM_Cubic, leaveTangent=-0.4588, tangentMode=RCTM_User, time=0.2, value=1}
         {arriveTangent=-0.0447, interpMode=RCIM_Cubic, tangentMode=RCTM_Break, time=1, value=0}
        ]}}})
    Scale RGB=&VectorFromFloat(Value=&MakeFloatFromVector4(
      Vector 4=&MakeVector4_FromFloat(Float=&FloatFromCurve(
        FloatCurve=DataInterface() {propertyValues {Curve {keys [<at 0 2>,<at 1 1.2>,<at 2 1>,<at 3 1>,<at 4 1>,<at 6 1>,<at 7 1.5>,<at 12 1.5>,<at 15 2>,<at 17 2>,<at 18 2>,<at 22 1.5>,<at 23 2>]}}}
        CurveIndex=&Multiply_Float(A=Linked() {linkedVariable {name=NPC.FortniteNPC.TimeOfDay_0-1, type=NiagaraFloat}}, B=24)
       ))
      Channel=::NewEnumerator0
     ))
    ScaleA=true
    ScaleRGB=true
   ) {enabled=false, module=ScaleColor003}
   &SolveForcesAndVelocity(Clamp Velocity=false, Limit Acceleration=false)
   &DynamicMaterialParameters(
    Index 0 Param 1=&FloatFromCurve(FloatCurve=DataInterface() {propertyValues {Curve {keys [
         {interpMode=RCIM_Cubic, leaveTangent=-0.7415, tangentMode=RCTM_Break, time=0, value=1}
         {arriveTangent=-1.5521, interpMode=RCIM_Cubic, leaveTangent=-8.0089, tangentMode=RCTM_Break, time=1, value=0}
        ]}}})
    Index 0 Param 2=&FloatFromCurve(
     FloatCurve=DataInterface() {propertyValues {Curve {keys [
         {interpMode=RCIM_Cubic, leaveTangent=-0.4439, tangentMode=RCTM_Break, time=0, value=0.2}
         {arriveTangent=-0.0393, interpMode=RCIM_Cubic, tangentMode=RCTM_Break, time=1, value=-0.3}
        ]}}}
     Scale Curve=0.7
    )
    Index 0 Param 3=&FloatFromCurve(
     FloatCurve=DataInterface() {propertyValues {Curve {keys [{interpMode=RCIM_Cubic, leaveTangent=2.6753, tangentMode=RCTM_Break, time=0.5, value=0}, <at 1 1>]}}}
     Scale Curve=3
    )
    Index 0 Param 4=&FloatFromCurve(FloatCurve=DataInterface() {propertyValues {Curve {keys [{interpMode=RCIM_Cubic, leaveTangent=2.6753, tangentMode=RCTM_Break, time=0, value=0}, <at 1 1>]}}})
    Index 1 Param 1=&FloatFromCurve(FloatCurve=DataInterface() {propertyValues {Curve {keys [{interpMode=RCIM_Cubic, tangentMode=RCTM_User, time=0, value=1}, {interpMode=RCIM_Cubic, tangentMode=RCTM_User, time=0.6, value=0}]}}})
    Module.Param0WriteEnabled=true
    Module.Param0WriteEnabled1=true
    Module.Param1WriteEnabled=true
    Module.Param1WriteEnabled1=true
    Module.Param2WriteEnabled=true
    Module.Param2WriteEnabled1=true
    Module.Param3WriteEnabled=true
    Module.Param3WriteEnabled1=true
    Write Parameter Index 1=true
   )
   &ScaleColor(
    Scale Alpha=&FNTimeOfDayScalar(Night Scale=0.8, Morning Scale=0.4, Evening Scale=0.5)
    Scale RGB=&MakeVectorFromLinearColorRGB(Color=&FNTimeOfDayColor(Night Color=<rgba 0.6 0.8 1 1>, Morning Color=<rgba 1 0.625 0.55 1>, Day Color=<rgba 1 1 1 1>, Evening Color=<rgba 1 0.5 0.4 1>))
    ScaleA=true
    ScaleRGB=true
   ) {enabled=false, module=ScaleColor001}
   &ScaleColor(
    Scale Alpha=$System.TOD_Alpha
    Scale RGB=&MakeVectorFromLinearColorRGB(Color=$System.TOD_Color)
    ScaleA=true
    ScaleRGB=true
   ) {module=ScaleColor002}
   &KillParticles(Kill Particles=&SetBoolByFloatComparison(A=$Particles.NormalizedAge, B=0.9))
  ]
  rendererClasses [&NiagaraSpriteRendererProperties]
  renderers [{rendererClass=&NiagaraSpriteRendererProperties}]
 }
 DropParticle {
  emitterUpdateScript [
   &EmitterState(
    Enable Distance Culling=true
    Life Cycle Mode=::NewEnumerator1
    Loop Duration=2
    MaxDistance=15000
    Module.bUseMaxDistance=true
    Module.bUseMinDistance=false
    Scalability Mode=::NewEnumerator1
    UseLoopDelay=false
   )
   &SpawnRate(
    SpawnRate=&Multiply_Float(A=7, B=&Divide_Float(A=$User.FallWidth, B=850))
    Use Spawn Probability=false
   )
  ]
  particleSpawnScript [
   &InitializeParticle(
    Color Minimum=<rgba 1 1 1 0.6>
    Color Mode=::NewEnumerator2
    Lifetime Max=3.5
    Lifetime Min=2.5
    Lifetime Mode=::NewEnumerator1
    Mass Mode=::NewEnumerator3
    Sprite Rotation Mode=::NewEnumerator2
    Sprite Size Mode=::NewEnumerator1
    Uniform Sprite Size Max=550
    Uniform Sprite Size Min=450
    UsePositionOffset=false
    Write Material Random=true
   )
   &ShapeLocation(
    Box Midpoint=<0.5 1 1>
    Box Size=<0 100 100>
    MaskSpawnGroup=false
    Offset Coordinate Space=::Local
    Offset Mode=::NewEnumerator0
    Override Seed=false
    Shape Origin=$User.TopPosition
    Shape Primitive=::NewEnumerator2
    Surface Only=false
    Surface Only Band Thickness=1
    UseOwnerScale=true
   )
   &AddVelocity(Velocity=<0 60 0>)
   /SetVariables(Particles.Position=&AddVectorToPosition(
     Position=$Particles.Position
     Vector=&Multiply_VectorByFloat(
      Vector=&TransformVector(Vector=<1 0 0>, Destination Space=::Simulation)
      Float=&Multiply_Float(
       A=&Lerp_Float(
        A=0
        B=&Divide_Float(A=$User.FallWidth, B=2)
        Alpha=&FloatFromCurve(
         FloatCurve=DataInterface() {propertyValues {Curve {keys [
             {interpMode=RCIM_Cubic, leaveTangent=1.3252, tangentMode=RCTM_Break, time=0, value=0}
             {arriveTangent=0.144, interpMode=RCIM_Cubic, tangentMode=RCTM_Break, time=1, value=1}
            ]}}}
         CurveIndex=&RandomRangeFloat(Minimum=0, Maximum=1)
        )
       )
       B=&Lerp_Float(
        A=-1
        B=1
        Alpha=&MakeFloatFromInt(Int=&RandomRangeInt(Minimum=0, Maximum=1))
       )
      )
     )
    )) {isSetParameters=true}
  ]
  particleUpdateScript [
   &ParticleState()
   &ScaleSpriteSize(Scale Sprite Size Mode=::NewEnumerator2) {module=ScaleSpriteSize001}
   &ScaleSpriteSize(Uniform Curve Sprite Scale=DataInterface() {propertyValues {Curve {keys [
        {arriveTangent=5.6242, interpMode=RCIM_Cubic, leaveTangent=5.6242, tangentMode=RCTM_User, time=0, value=0}
        {arriveTangent=0.9995, interpMode=RCIM_Cubic, leaveTangent=0.9995, tangentMode=RCTM_User, time=0.5, value=1}
       ]}}})
   &GravityForce(Gravity=&RandomRangeVector(Minimum=<0 0 -204.1991>, Maximum=<1 1 -457.8731>))
   &Drag(Drag=0.5, Use Linear Drag=true, Use Rotational Drag=false)
   &ScaleColor(
    Scale Alpha=&FloatFromCurve(FloatCurve=DataInterface() {propertyValues {Curve {keys [
         {arriveTangent=6.213, interpMode=RCIM_Cubic, leaveTangent=6.213, tangentMode=RCTM_User, time=0, value=0}
         {arriveTangent=-1.5422, interpMode=RCIM_Cubic, leaveTangent=-1.5422, tangentMode=RCTM_User, time=0.4, value=1.075}
         {arriveTangent=-1.4513, interpMode=RCIM_Cubic, tangentMode=RCTM_Break, time=1, value=0}
        ]}}})
    ScaleA=true
    ScaleRGB=true
   )
   &CurlNoiseForce(
    Noise Frequency=0.5
    Noise Strength=&FloatFromCurve(
     FloatCurve=DataInterface() {propertyValues {Curve {keys [{tangentMode=RCTM_User, time=0, value=0}, {tangentMode=RCTM_User, time=1, value=1}]}}}
     Scale Curve=400
    )
    Pan Noise=true
    Pan Noise Field=<0 0 -0.3>
   )
   &SolveForcesAndVelocity(Clamp Velocity=false, Limit Acceleration=false)
   &ScaleColor(
    Scale Alpha=&FNTimeOfDayScalar(Night Scale=1, Morning Scale=0.2, Evening Scale=0.2)
    Scale RGB=&MakeVectorFromLinearColorRGB(Color=&FNTimeOfDayColor(Night Color=<rgba 0.6 0.8 1 1>, Morning Color=<rgba 1 0.625 0.55 1>, Day Color=<rgba 1 1 1 1>, Evening Color=<rgba 1 0.5 0.4 1>))
    ScaleA=true
    ScaleRGB=true
   ) {enabled=false, module=ScaleColor001}
   &ScaleColor(
    Scale Alpha=$System.TOD_Alpha
    Scale RGB=&MakeVectorFromLinearColorRGB(Color=$System.TOD_Color)
    ScaleA=true
    ScaleRGB=true
   ) {module=ScaleColor002}
   &KillParticles(Kill Particles=&SetBoolByFloatComparison(A=$Particles.NormalizedAge, B=0.8))
  ]
  rendererClasses [&NiagaraSpriteRendererProperties]
  renderers [{rendererClass=&NiagaraSpriteRendererProperties}]
 }
 DropSmoke {
  emitterUpdateScript [
   &EmitterState(
    Enable Distance Culling=true
    Life Cycle Mode=::NewEnumerator1
    Loop Duration=2
    MaxDistance=15000
    Module.bUseMaxDistance=true
    Module.bUseMinDistance=false
    Scalability Mode=::NewEnumerator1
    UseLoopDelay=false
   )
   &SpawnRate(
    SpawnRate=&Multiply_Float(A=13, B=&Divide_Float(A=$User.FallWidth, B=850))
    Use Spawn Probability=false
   )
  ]
  particleSpawnScript [
   &InitializeParticle(
    Color Maximum=<rgba 1 1 1 0.55>
    Color Minimum=<rgba 1 1 1 0.4>
    Color Mode=::NewEnumerator2
    Lifetime Max=6
    Lifetime Min=4
    Lifetime Mode=::NewEnumerator1
    Mass Mode=::NewEnumerator3
    Sprite Rotation Mode=::NewEnumerator2
    Sprite Size Mode=::NewEnumerator1
    Uniform Sprite Size Max=500
    Uniform Sprite Size Min=200
    UsePositionOffset=false
    Write Material Random=true
   )
   &ShapeLocation(
    Box Size=<0 100 100>
    MaskSpawnGroup=false
    Offset=<0 -50 0>
    Offset Coordinate Space=::Local
    Offset Mode=::NewEnumerator0
    Override Seed=false
    Shape Origin=$User.TopPosition
    Shape Primitive=::NewEnumerator2
    Surface Only=false
    Surface Only Band Thickness=1
    UseOwnerScale=true
   )
   &AddVelocity(Velocity=<0 60 0>)
   &SubUVAnimation(End Frame=3, SubUV Animation Mode:ENiagara_SubUVLookupMode=::NewEnumerator2)
   /SetVariables(Particles.Position=&AddVectorToPosition(
     Position=$Particles.Position
     Vector=&Multiply_VectorByFloat(
      Vector=&TransformVector(Vector=<1 0 0>, Destination Space=::Simulation)
      Float=&Multiply_Float(
       A=&Lerp_Float(
        A=0
        B=&Divide_Float(A=$User.FallWidth, B=2)
        Alpha=&FloatFromCurve(
         FloatCurve=DataInterface() {propertyValues {Curve {keys [
             {interpMode=RCIM_Cubic, leaveTangent=1.3252, tangentMode=RCTM_Break, time=0, value=0}
             {arriveTangent=0.144, interpMode=RCIM_Cubic, tangentMode=RCTM_Break, time=1, value=1}
            ]}}}
         CurveIndex=&RandomRangeFloat(Minimum=0, Maximum=1)
        )
       )
       B=&Lerp_Float(
        A=-1
        B=1
        Alpha=&MakeFloatFromInt(Int=&RandomRangeInt(Minimum=0, Maximum=1))
       )
      )
     )
    )) {isSetParameters=true}
  ]
  particleUpdateScript [
   &ParticleState()
   &ScaleSpriteSize(Scale Factor=<1 2>, Scale Sprite Size Mode=::NewEnumerator2) {module=ScaleSpriteSize001}
   &ScaleSpriteSize(Uniform Curve Sprite Scale=DataInterface() {propertyValues {Curve {keys [
        {arriveTangent=5.6242, interpMode=RCIM_Cubic, leaveTangent=5.6242, tangentMode=RCTM_User, time=0, value=0}
        {arriveTangent=0.9995, interpMode=RCIM_Cubic, leaveTangent=0.9995, tangentMode=RCTM_User, time=0.5, value=1}
       ]}}})
   &GravityForce(Gravity=&RandomRangeVector(Minimum=<0 0 -150>, Maximum=<1 1 -350>))
   &Drag(Drag=0.5, Use Linear Drag=true, Use Rotational Drag=false)
   &ScaleColor(
    Scale Alpha=&FloatFromCurve(FloatCurve=DataInterface() {propertyValues {Curve {keys [
         {arriveTangent=6.213, interpMode=RCIM_Cubic, leaveTangent=6.213, tangentMode=RCTM_User, time=0, value=0}
         {arriveTangent=-1.5422, interpMode=RCIM_Cubic, leaveTangent=-1.5422, tangentMode=RCTM_User, time=0.4, value=1.075}
         {arriveTangent=-1.4513, interpMode=RCIM_Cubic, tangentMode=RCTM_Break, time=1, value=0}
        ]}}})
    ScaleA=true
    ScaleRGB=true
   )
   &CurlNoiseForce(
    Noise Frequency=1
    Noise Strength=&FloatFromCurve(
     FloatCurve=DataInterface() {propertyValues {Curve {keys [{tangentMode=RCTM_User, time=0, value=0}, {tangentMode=RCTM_User, time=1, value=1}]}}}
     Scale Curve=300
    )
    Pan Noise=true
    Pan Noise Field=<0 0 -0.3>
   )
   &SolveForcesAndVelocity(Clamp Velocity=false, Limit Acceleration=false)
   &ScaleColor(
    Scale Alpha=&FNTimeOfDayScalar(Night Scale=1.25, Morning Scale=0.5, Evening Scale=0.5)
    Scale RGB=&MakeVectorFromLinearColorRGB(Color=&FNTimeOfDayColor(Night Color=<rgba 0.6 0.8 1 1>, Morning Color=<rgba 1 0.625 0.55 1>, Day Color=<rgba 1 1 1 1>, Evening Color=<rgba 1 0.5 0.4 1>))
    ScaleA=true
    ScaleRGB=true
   ) {enabled=false, module=ScaleColor001}
   &ScaleColor(
    Scale Alpha=$System.TOD_Alpha
    Scale RGB=&MakeVectorFromLinearColorRGB(Color=$System.TOD_Color)
    ScaleA=true
    ScaleRGB=true
   ) {module=ScaleColor002}
   &KillParticles(Kill Particles=&SetBoolByFloatComparison(A=$Particles.NormalizedAge, B=0.9))
  ]
  rendererClasses [&NiagaraSpriteRendererProperties]
  renderers [{rendererClass=&NiagaraSpriteRendererProperties}]
 }
 DropSplash {
  emitterUpdateScript [
   &EmitterState(
    Enable Distance Culling=true
    Life Cycle Mode=::NewEnumerator1
    Loop Duration=2
    MaxDistance=12500
    Module.bUseMaxDistance=true
    Module.bUseMinDistance=false
    Scalability Mode=::NewEnumerator1
    UseLoopDelay=false
   )
   &SpawnRate(
    SpawnRate=&Multiply_Float(A=20, B=&Divide_Float(A=$User.FallWidth, B=850))
    Use Spawn Probability=false
   )
  ]
  particleSpawnScript [
   &InitializeParticle(
    Color Maximum=<rgba 1 1 1 0.3>
    Color Minimum=<rgba 1 1 1 0.2>
    Color Mode=::NewEnumerator2
    Lifetime Max=5
    Lifetime Min=2
    Lifetime Mode=::NewEnumerator1
    Mass Mode=::NewEnumerator3
    Sprite Rotation Mode=::NewEnumerator2
    Sprite Size Mode=::NewEnumerator1
    Uniform Sprite Size Max=400
    Uniform Sprite Size Min=150
    UsePositionOffset=false
    Write Material Random=true
   )
   &ShapeLocation(
    Box Size=<0 100 100>
    MaskSpawnGroup=false
    Offset=<0 -80 0>
    Offset Coordinate Space=::Local
    Offset Mode=::NewEnumerator0
    Override Seed=false
    Shape Origin=$User.TopPosition
    Shape Primitive=::NewEnumerator2
    Surface Only=false
    Surface Only Band Thickness=1
    UseOwnerScale=true
   )
   &AddVelocity(Velocity=<0 100 0>)
   /SetVariables(Particles.Position=&AddVectorToPosition(
     Position=$Particles.Position
     Vector=&Multiply_VectorByFloat(
      Vector=&TransformVector(Vector=<1 0 0>, Destination Space=::Simulation)
      Float=&Multiply_Float(
       A=&Lerp_Float(
        A=0
        B=&Divide_Float(A=$User.FallWidth, B=2)
        Alpha=&FloatFromCurve(
         FloatCurve=DataInterface() {propertyValues {Curve {keys [
             {interpMode=RCIM_Cubic, leaveTangent=2.1204, tangentMode=RCTM_Break, time=0, value=0}
             {arriveTangent=0.321, interpMode=RCIM_Cubic, leaveTangent=0.321, tangentMode=RCTM_User, time=0.3, value=0.775}
             {arriveTangent=0.4955, interpMode=RCIM_Cubic, tangentMode=RCTM_Break, time=1, value=1}
            ]}}}
         CurveIndex=&RandomRangeFloat(Minimum=0, Maximum=1)
        )
       )
       B=&Lerp_Float(
        A=-1
        B=1
        Alpha=&MakeFloatFromInt(Int=&RandomRangeInt(Minimum=0, Maximum=1))
       )
      )
     )
    )) {isSetParameters=true}
   &SubUVAnimation(End Frame=3, SubUV Animation Mode:ENiagara_SubUVLookupMode=::NewEnumerator2)
  ]
  particleUpdateScript [
   &ParticleState()
   &ScaleSpriteSize(Scale Factor=<1 2>, Scale Sprite Size Mode=::NewEnumerator2) {module=ScaleSpriteSize001}
   &ScaleSpriteSize(Uniform Curve Sprite Scale=DataInterface() {propertyValues {Curve {keys [
        {arriveTangent=5.6242, interpMode=RCIM_Cubic, leaveTangent=5.6242, tangentMode=RCTM_User, time=0, value=0}
        {arriveTangent=0.9995, interpMode=RCIM_Cubic, leaveTangent=0.9995, tangentMode=RCTM_User, time=0.5, value=1}
       ]}}})
   &GravityForce(Gravity=&RandomRangeVector(Minimum=<0 0 -300>, Maximum=<1 1 -600>))
   &Drag(Drag=0.3, Use Linear Drag=true, Use Rotational Drag=false)
   &ScaleColor(
    Scale Alpha=&FloatFromCurve(FloatCurve=DataInterface() {propertyValues {Curve {keys [
         {arriveTangent=6.213, interpMode=RCIM_Cubic, leaveTangent=6.213, tangentMode=RCTM_User, time=0, value=0}
         {arriveTangent=-1.5422, interpMode=RCIM_Cubic, leaveTangent=-1.5422, tangentMode=RCTM_User, time=0.4, value=1.075}
         {arriveTangent=-1.4513, interpMode=RCIM_Cubic, tangentMode=RCTM_Break, time=1, value=0}
        ]}}})
    ScaleA=true
    ScaleRGB=true
   )
   &SolveForcesAndVelocity(Clamp Velocity=false, Limit Acceleration=false)
   &DynamicMaterialParameters(
    Index 0 Param 1=&FloatFromCurve(FloatCurve=DataInterface() {propertyValues {Curve {keys [{interpMode=RCIM_Cubic, leaveTangent=1.7548, tangentMode=RCTM_Break, time=0, value=0}, {tangentMode=RCTM_User, time=1, value=1}]}}})
    Module.Param0WriteEnabled=true
    Module.Param1WriteEnabled=true
    Module.Param2WriteEnabled=true
    Module.Param3WriteEnabled=true
   )
   &ScaleColor(
    Scale Alpha=&FNTimeOfDayScalar(Night Scale=1.2, Morning Scale=0.5, Evening Scale=0.5)
    Scale RGB=&MakeVectorFromLinearColorRGB(Color=&FNTimeOfDayColor(Night Color=<rgba 0.6 0.8 1 1>, Morning Color=<rgba 1 0.625 0.55 1>, Day Color=<rgba 1 1 1 1>, Evening Color=<rgba 1 0.5 0.4 1>))
    ScaleA=true
    ScaleRGB=true
   ) {enabled=false, module=ScaleColor001}
   &ScaleColor(
    Scale Alpha=$System.TOD_Alpha
    Scale RGB=&MakeVectorFromLinearColorRGB(Color=$System.TOD_Color)
    ScaleA=true
    ScaleRGB=true
   ) {module=ScaleColor002}
   &KillParticles(Kill Particles=&SetBoolByFloatComparison(A=$Particles.NormalizedAge, B=0.8))
  ]
  rendererClasses [&NiagaraSpriteRendererProperties]
  renderers [{rendererClass=&NiagaraSpriteRendererProperties}]
 }
 EdgeFlow {
  emitterUpdateScript [
   &EmitterState(
    Inactive Response=::NewEnumerator1
    Life Cycle Mode=::NewEnumerator1
    Loop Behavior=::NewEnumerator1
    Loop Duration Mode=::NewEnumerator1
    UseLoopDelay=false
   )
   &SpawnBurst_Instantaneous(Spawn Count=2, Use Spawn Probability=false, UseLoopCountLimit=false)
  ]
  particleSpawnScript [
   &FN_SetSpawnGroup(Number of Spawn Groups=2)
   &InitializeParticle(
    Lifetime=5
    Position=&AddVectorToPosition(
     Position=$User.TopPosition
     Vector=&TransformVector(
      Vector=&MakeVector(
       X=&Divide_Float(A=$User.FallWidth, B=&MaskFloatBySpawnGroup(Masked Float=-2, Passthrough Float=2))
       Y=0
       Z=0
      )
      Destination Space=::World
     )
    )
    Position Mode=::NewEnumerator1
    Position Offset=&MakeVector(X=&MaskFloatBySpawnGroup(Masked Float=80, Passthrough Float=-80), Y=-20, Z=0)
    Position Offset Coordinate Space=::Local
    Sprite Rotation Angle=-90
    Sprite Rotation Mode=::NewEnumerator0
    Sprite Size Max=<1300 400>
    Sprite Size Min=<1000 300>
    Sprite Size Mode=::NewEnumerator2
    UsePositionOffset=true
    Write Material Random=true
   )
  ]
  particleUpdateScript [
   &ParticleState(Kill Particles When Lifetime Has Elapsed=false, Let Infinitely Lived Particles Die When Emitter Deactivates=true)
   &SpriteFacingAndAlignment(Sprite Alignment=<0 0 1>, WriteAlignment=true, WriteFacing=false)
   &ScaleColor(
    Scale Alpha=&FNTimeOfDayScalar(Night Scale=1, Morning Scale=0.5, Evening Scale=0.5)
    Scale RGB=&MakeVectorFromLinearColorRGB(Color=&FNTimeOfDayColor(Night Color=<rgba 0.6 0.8 1 1>, Morning Color=<rgba 1 0.625 0.55 1>, Day Color=<rgba 1 1 1 1>, Evening Color=<rgba 1 0.5 0.4 1>))
    ScaleA=true
    ScaleRGB=true
   ) {enabled=false, module=ScaleColor001}
   &ScaleColor(
    Scale Alpha=$System.TOD_Alpha
    Scale RGB=&MakeVectorFromLinearColorRGB(Color=$System.TOD_Color)
    ScaleA=true
    ScaleRGB=true
   ) {module=ScaleColor002}
  ]
  rendererClasses [&NiagaraSpriteRendererProperties]
  renderers [{rendererClass=&NiagaraSpriteRendererProperties}]
 }
 Fountain {
  emitterUpdateScript [
   &EmitterState(
    Enable Distance Culling=true
    Life Cycle Mode=::NewEnumerator1
    Loop Duration=2
    MaxDistance=10000
    Module.bUseMaxDistance=true
    Module.bUseMinDistance=false
    Scalability Mode=::NewEnumerator1
    UseLoopDelay=false
   )
   /SetVariables(Emitter.New float:NiagaraFloat=&FloatFromCurve(FloatCurve=DataInterface() {propertyValues {Curve {keys [<at 0 0>,<at 1 1>]}}}, CurveIndex=$User.HitSplashEffectDensity, Scale Curve=10)) {isSetParameters=true}
   &self:CalcSpawnRateFromArray(BaseRate:NiagaraFloat=$Emitter.New float, SpawnPosArray:NiagaraDataInterfaceArrayPosition=$User.SpawnSplashPositionArray) {module=CalcSpawnRateFromArray}
   &SpawnRate(SpawnRate=$Emitter.CalcSpawnRateFromArray.SpawnRate, Use Spawn Probability=false)
  ]
  particleSpawnScript [
   &FN_SetSpawnGroup(Number of Spawn Groups:NiagaraInt32=$Emitter.CalcSpawnRateFromArray.ArraySize)
   &InitializeParticle(
    Color Minimum=<rgba 1 1 1 1>
    Color Mode=::NewEnumerator2
    Lifetime Max=1.8
    Lifetime Min=1.2
    Lifetime Mode=::NewEnumerator1
    Mass Mode=::NewEnumerator3
    Sprite Rotation Mode=::NewEnumerator2
    Sprite Size Mode=::NewEnumerator1
    Uniform Sprite Size Max=220
    Uniform Sprite Size Min=120
    UsePositionOffset=false
    Write Material Random=true
   )
   &self:GetAndSetPositionFromArray(SpawnGroup:NiagaraInt32=$Particles.SpawnGroup, SpawnPosArray:NiagaraDataInterfaceArrayPosition=$User.SpawnSplashPositionArray) {module=GetAndSetPositionFromArray}
   &ShapeLocation(
    MaskSpawnGroup=false
    Override Seed=false
    Shape Origin=$Particles.GetAndSetPositionFromArray.ArrayPosition
    Sphere Radius=30
    UseOwnerScale=true
   )
   &AddVelocity(
    Cone Angle=32
    Cone Axis=<0 0.2 1>
    Distribution Along Cone Axis=0.25
    Module.UseConeFalloff=true
    Module.UseVelDistribution=true
    Override Seed=false
    Speed Falloff From Cone Axis=0.5
    Velocity Mode=::NewEnumerator2
    Velocity Speed=&RandomRangeFloat(Minimum=443.697, Maximum=767.4904)
   )
   &SubUVAnimation(End Frame=3)
  ]
  particleUpdateScript [
   &ParticleState()
   &SpriteRotationRate(Rotation Rate=50)
   &ScaleSpriteSize(Uniform Curve Sprite Scale=DataInterface() {propertyValues {Curve {keys [
        {interpMode=RCIM_Cubic, leaveTangent=5.6407, tangentMode=RCTM_Break, time=0, value=0.4}
        {tangentMode=RCTM_User, time=0.3, value=1}
        <at 1 1.5>
       ]}}})
   &GravityForce(Gravity=<0 0 -1200>)
   &ScaleColor(
    Scale Alpha=&FloatFromCurve(FloatCurve=DataInterface() {propertyValues {Curve {keys [
         {arriveTangent=13.2056, interpMode=RCIM_Cubic, leaveTangent=13.2056, tangentMode=RCTM_User, time=0, value=0}
         {arriveTangent=-0.4588, interpMode=RCIM_Cubic, leaveTangent=-0.4588, tangentMode=RCTM_User, time=0.2, value=1}
         {arriveTangent=-0.0447, interpMode=RCIM_Cubic, tangentMode=RCTM_Break, time=1, value=0}
        ]}}})
    ScaleA=true
    ScaleRGB=true
   )
   &SolveForcesAndVelocity(Clamp Velocity=false, Limit Acceleration=false)
   &DynamicMaterialParameters(
    Index 0 Param 1=&FloatFromCurve(FloatCurve=DataInterface() {propertyValues {Curve {keys [
         {interpMode=RCIM_Cubic, leaveTangent=-0.7415, tangentMode=RCTM_Break, time=0, value=1}
         {arriveTangent=-1.5521, interpMode=RCIM_Cubic, leaveTangent=-8.0089, tangentMode=RCTM_Break, time=1, value=0}
        ]}}})
    Index 0 Param 2=&FloatFromCurve(
     FloatCurve=DataInterface() {propertyValues {Curve {keys [
         {interpMode=RCIM_Cubic, leaveTangent=-0.4439, tangentMode=RCTM_Break, time=0, value=0.3}
         {arriveTangent=-0.0393, interpMode=RCIM_Cubic, tangentMode=RCTM_Break, time=1, value=-0.3}
        ]}}}
     Scale Curve=0.7
    )
    Index 0 Param 3=&FloatFromCurve(
     FloatCurve=DataInterface() {propertyValues {Curve {keys [{interpMode=RCIM_Cubic, leaveTangent=2.6753, tangentMode=RCTM_Break, time=0.5, value=0}, <at 1 1>]}}}
     Scale Curve=3
    )
    Index 0 Param 4=&FloatFromCurve(FloatCurve=DataInterface() {propertyValues {Curve {keys [{interpMode=RCIM_Cubic, leaveTangent=2.6753, tangentMode=RCTM_Break, time=0, value=0}, <at 1 1>]}}})
    Index 1 Param 1=&FloatFromCurve(FloatCurve=DataInterface() {propertyValues {Curve {keys [{interpMode=RCIM_Cubic, tangentMode=RCTM_User, time=0, value=1}, {interpMode=RCIM_Cubic, tangentMode=RCTM_User, time=0.6, value=0}]}}})
    Module.Param0WriteEnabled=true
    Module.Param0WriteEnabled1=true
    Module.Param1WriteEnabled=true
    Module.Param1WriteEnabled1=true
    Module.Param2WriteEnabled=true
    Module.Param2WriteEnabled1=true
    Module.Param3WriteEnabled=true
    Module.Param3WriteEnabled1=true
    Write Parameter Index 1=true
   )
   &ScaleColor(
    Scale Alpha=$System.TOD_Alpha
    Scale RGB=&MakeVectorFromLinearColorRGB(Color=$System.TOD_Color)
    ScaleA=true
    ScaleRGB=true
   ) {module=ScaleColor001}
   &KillParticles(Kill Particles=&SetBoolByFloatComparison(A=$Particles.NormalizedAge, B=0.9))
  ]
  rendererClasses [&NiagaraSpriteRendererProperties]
  renderers [{rendererClass=&NiagaraSpriteRendererProperties}]
 }
 TopSmoke {
  emitterUpdateScript [
   &EmitterState(
    Enable Distance Culling=true
    MaxDistance=20000
    Module.bUseMaxDistance=true
    Module.bUseMinDistance=false
    Scalability Mode=::NewEnumerator1
   )
   &SpawnRate(
    SpawnRate=&Multiply_Float(A=7, B=&Divide_Float(A=$User.FallWidth, B=850))
    Use Spawn Probability=false
   )
  ]
  particleSpawnScript [
   &InitializeParticle(
    Color=<rgba 1 1 1 0.35>
    Lifetime Max=2.2
    Lifetime Min=2
    Lifetime Mode=::NewEnumerator1
    Sprite Size Mode=::NewEnumerator3
    Uniform Sprite Size=&Lerp_Float(
     A=500
     B=800
     Alpha=&FloatFromCurve(
      FloatCurve=DataInterface() {propertyValues {Curve {keys [
          {interpMode=RCIM_Cubic, leaveTangent=0.1305, tangentMode=RCTM_Break, time=0, value=0}
          {arriveTangent=1, interpMode=RCIM_Cubic, leaveTangent=1, time=0.7, value=0.3}
          {arriveTangent=2.0211, interpMode=RCIM_Cubic, tangentMode=RCTM_Break, time=1, value=1}
         ]}}}
      CurveIndex=&RandomRangeFloat(Minimum=0, Maximum=1)
     )
    )
    UsePositionOffset=false
    Write Material Random=true
   )
   &ShapeLocation(
    Box Midpoint=<0.5 1 0.5>
    Box Size=<0 200 200>
    MaskSpawnGroup=false
    Offset=<0 -300 -50>
    Offset Coordinate Space=::Local
    Offset Mode=::NewEnumerator0
    Override Seed=false
    Shape Origin=$User.TopPosition
    Shape Primitive=::NewEnumerator2
    Surface Only=false
    UseOwnerScale=true
   )
   /SetVariables(Particles.Position=&AddVectorToPosition(
     Position=$Particles.Position
     Vector=&Multiply_VectorByFloat(
      Vector=&TransformVector(Vector=<1 0 0>, Destination Space=::Simulation)
      Float=&Multiply_Float(
       A=&Lerp_Float(
        A=0
        B=&Divide_Float(A=$User.FallWidth, B=2)
        Alpha=&FloatFromCurve(
         FloatCurve=DataInterface() {propertyValues {Curve {keys [
             {interpMode=RCIM_Cubic, leaveTangent=3.3639, tangentMode=RCTM_Break, time=0, value=0}
             {arriveTangent=-0.0122, interpMode=RCIM_Cubic, tangentMode=RCTM_Break, time=1, value=1}
            ]}}}
         CurveIndex=&RandomRangeFloat(Minimum=0, Maximum=1)
        )
       )
       B=&Lerp_Float(
        A=-1
        B=1
        Alpha=&MakeFloatFromInt(Int=&RandomRangeInt(Minimum=0, Maximum=1))
       )
      )
     )
    )) {isSetParameters=true}
   /SetVariables(Particles.Velocity=&Multiply_Vector(A=$Particles.Velocity, B=<1 1 0.2>)) {isSetParameters=true, module=SetVariables_#5}
   &AddVelocity(Velocity=&RandomRangeVector(Minimum=<0 368.6783 -100>, Maximum=<1 445.846 -20>))
  ]
  particleUpdateScript [
   &ParticleState()
   &SpriteRotationRate(Rotation Rate=-50)
   &ScaleColor(
    Scale Alpha=&FloatFromCurve(FloatCurve=DataInterface() {propertyValues {Curve {keys [
         {interpMode=RCIM_Cubic, tangentMode=RCTM_User, time=0, value=0}
         {interpMode=RCIM_Cubic, tangentMode=RCTM_User, time=0.5, value=1}
         {tangentMode=RCTM_User, time=1, value=0}
        ]}}})
    ScaleA=true
    ScaleRGB=true
   )
   &GravityForce(Gravity=&RandomRangeVector(Minimum=<0 0 -50>, Maximum=<0 0 100>))
   &Drag(Drag=0.1636, Use Linear Drag=true, Use Rotational Drag=true)
   &SolveForcesAndVelocity(Clamp Velocity=false, Limit Acceleration=false)
   &DynamicMaterialParameters(
    Index 0 Param 1=&RandomRangeFloat(Minimum=0, Maximum=1)
    Module.Param0WriteEnabled=true
    Module.Param1WriteEnabled=true
    Module.Param2WriteEnabled=true
    Module.Param3WriteEnabled=true
   )
   &SubUVAnimation() {enabled=false, module=SubUVAnimation001}
   &ScaleColor(
    Scale Alpha=&FNTimeOfDayScalar(Night Scale=0.5, Morning Scale=0.2, Evening Scale=0.3)
    Scale RGB=&MakeVectorFromLinearColorRGB(Color=&FNTimeOfDayColor(Night Color=<rgba 0.6 0.8 1 1>, Morning Color=<rgba 1 0.625 0.55 1>, Day Color=<rgba 1 1 1 1>, Evening Color=<rgba 1 0.5 0.4 1>))
    ScaleA=true
    ScaleRGB=true
   ) {enabled=false, module=ScaleColor001}
   &ScaleColor(
    Scale Alpha=$System.TOD_Alpha
    Scale RGB=&MakeVectorFromLinearColorRGB(Color=$System.TOD_Color)
    ScaleA=true
    ScaleRGB=true
   ) {module=ScaleColor002}
   &KillParticles(Kill Particles=&SetBoolByFloatComparison(A=$Particles.NormalizedAge, B=0.95))
  ]
  rendererClasses [&NiagaraSpriteRendererProperties]
  renderers [{rendererClass=&NiagaraSpriteRendererProperties}]
 }
 TopSplash {
  emitterUpdateScript [
   &EmitterState(
    Enable Distance Culling=true
    Life Cycle Mode=::NewEnumerator1
    Loop Duration=2
    MaxDistance=10000
    Module.bUseMaxDistance=true
    Module.bUseMinDistance=false
    Scalability Mode=::NewEnumerator1
    UseLoopDelay=false
   )
   &SpawnRate(
    SpawnRate=&Multiply_Float(A=$User.TopSplashRate, B=&Divide_Float(A=$User.FallWidth, B=850))
    Use Spawn Probability=false
   )
  ]
  particleSpawnScript [
   &InitializeParticle(
    Color Maximum=<rgba 1 1 1 1.5>
    Color Minimum=<rgba 1 1 1 1>
    Color Mode=::NewEnumerator2
    Lifetime Max=1.3
    Lifetime Min=0.6
    Lifetime Mode=::NewEnumerator1
    Mass Mode=::NewEnumerator3
    Sprite Rotation Mode=::NewEnumerator2
    Sprite Size Mode=::NewEnumerator1
    Uniform Sprite Size Max=223.614
    Uniform Sprite Size Min=161.7636
    UsePositionOffset=false
    Write Material Random=true
   )
   &ShapeLocation(
    Box Size=<0 50 50>
    MaskSpawnGroup=false
    Offset=<0 -100 -75>
    Offset Coordinate Space=::Local
    Offset Mode=::NewEnumerator0
    Override Seed=false
    Shape Origin=$User.TopPosition
    Shape Primitive=::NewEnumerator2
    Surface Only=false
    UseOwnerScale=true
   )
   /SetVariables(Particles.Position=&AddVectorToPosition(
     Position=$Particles.Position
     Vector=&Multiply_VectorByFloat(
      Vector=&TransformVector(Vector=<1 0 0>, Destination Space=::Simulation)
      Float=&Multiply_Float(
       A=&Lerp_Float(
        A=0
        B=&Divide_Float(A=$User.FallWidth, B=2)
        Alpha=&FloatFromCurve(
         FloatCurve=DataInterface() {propertyValues {Curve {keys [
             {interpMode=RCIM_Cubic, leaveTangent=3.3639, tangentMode=RCTM_Break, time=0, value=0}
             {arriveTangent=-0.0122, interpMode=RCIM_Cubic, tangentMode=RCTM_Break, time=1, value=1}
            ]}}}
         CurveIndex=&RandomRangeFloat(Minimum=0, Maximum=1)
        )
       )
       B=&Lerp_Float(
        A=-1
        B=1
        Alpha=&MakeFloatFromInt(Int=&RandomRangeInt(Minimum=0, Maximum=1))
       )
      )
     )
    )) {isSetParameters=true, module=SetVariables_#6}
   &AddVelocity(
    Cone Angle=32
    Cone Axis=<0 0.5 1>
    Distribution Along Cone Axis=0.25
    Module.UseConeFalloff=true
    Module.UseVelDistribution=true
    Override Seed=false
    Speed Falloff From Cone Axis=0.5
    Velocity Mode=::NewEnumerator2
    Velocity Speed=&RandomRangeFloat(Minimum=383.6636, Maximum=646.2988)
   )
   &SubUVAnimation(End Frame=3)
  ]
  particleUpdateScript [
   &ParticleState()
   &ScaleSpriteSize(
    Uniform Curve Scale=$User.TopSplashScale
    Uniform Curve Sprite Scale=DataInterface() {propertyValues {Curve {keys [
        {interpMode=RCIM_Cubic, leaveTangent=5.6407, tangentMode=RCTM_Break, time=0, value=0.4}
        {tangentMode=RCTM_User, time=0.3, value=1}
        <at 1 1.5>
       ]}}}
   )
   &GravityForce(Gravity=<0 0 -1200>)
   &ScaleColor(
    Scale Alpha=&FloatFromCurve(FloatCurve=DataInterface() {propertyValues {Curve {keys [
         {arriveTangent=13.2056, interpMode=RCIM_Cubic, leaveTangent=13.2056, tangentMode=RCTM_User, time=0, value=0}
         {arriveTangent=-0.4588, interpMode=RCIM_Cubic, leaveTangent=-0.4588, tangentMode=RCTM_User, time=0.2, value=1}
         {arriveTangent=-0.0447, interpMode=RCIM_Cubic, tangentMode=RCTM_Break, time=1, value=0}
        ]}}})
    ScaleA=true
    ScaleRGB=true
   )
   &SolveForcesAndVelocity(Clamp Velocity=false, Limit Acceleration=false)
   &DynamicMaterialParameters(
    Index 0 Param 1=&FloatFromCurve(FloatCurve=DataInterface() {propertyValues {Curve {keys [
         {interpMode=RCIM_Cubic, leaveTangent=-0.7415, tangentMode=RCTM_Break, time=0, value=1}
         {arriveTangent=-1.5521, interpMode=RCIM_Cubic, leaveTangent=-8.0089, tangentMode=RCTM_Break, time=1, value=0}
        ]}}})
    Index 0 Param 2=&FloatFromCurve(
     FloatCurve=DataInterface() {propertyValues {Curve {keys [
         {interpMode=RCIM_Cubic, leaveTangent=-0.4439, tangentMode=RCTM_Break, time=0, value=0.2}
         {arriveTangent=-0.0393, interpMode=RCIM_Cubic, tangentMode=RCTM_Break, time=1, value=-0.3}
        ]}}}
     Scale Curve=0.7
    )
    Index 0 Param 3=&FloatFromCurve(
     FloatCurve=DataInterface() {propertyValues {Curve {keys [{interpMode=RCIM_Cubic, leaveTangent=2.6753, tangentMode=RCTM_Break, time=0.5, value=0}, <at 1 1>]}}}
     Scale Curve=3
    )
    Index 0 Param 4=&FloatFromCurve(FloatCurve=DataInterface() {propertyValues {Curve {keys [{interpMode=RCIM_Cubic, leaveTangent=2.6753, tangentMode=RCTM_Break, time=0, value=0}, <at 1 1>]}}})
    Index 1 Param 1=&FloatFromCurve(FloatCurve=DataInterface() {propertyValues {Curve {keys [{interpMode=RCIM_Cubic, tangentMode=RCTM_User, time=0, value=1}, {interpMode=RCIM_Cubic, tangentMode=RCTM_User, time=0.6, value=0}]}}})
    Module.Param0WriteEnabled=true
    Module.Param0WriteEnabled1=true
    Module.Param1WriteEnabled=true
    Module.Param1WriteEnabled1=true
    Module.Param2WriteEnabled=true
    Module.Param2WriteEnabled1=true
    Module.Param3WriteEnabled=true
    Module.Param3WriteEnabled1=true
    Write Parameter Index 1=true
   )
   &ScaleColor(
    Scale Alpha=&FNTimeOfDayScalar(Night Scale=0.8, Morning Scale=0.4, Evening Scale=0.5)
    Scale RGB=&MakeVectorFromLinearColorRGB(Color=&FNTimeOfDayColor(Night Color=<rgba 0.6 0.8 1 1>, Morning Color=<rgba 1 0.625 0.55 1>, Day Color=<rgba 1 1 1 1>, Evening Color=<rgba 1 0.5 0.4 1>))
    ScaleA=true
    ScaleRGB=true
   ) {enabled=false, module=ScaleColor001}
   &ScaleColor(
    Scale Alpha=$System.TOD_Alpha
    Scale RGB=&MakeVectorFromLinearColorRGB(Color=$System.TOD_Color)
    ScaleA=true
    ScaleRGB=true
   ) {module=ScaleColor002}
   &KillParticles(Kill Particles=&SetBoolByFloatComparison(A=$Particles.NormalizedAge, B=0.9))
  ]
  rendererClasses [&NiagaraSpriteRendererProperties]
  renderers [{rendererClass=&NiagaraSpriteRendererProperties}]
 }
}
summary {
 systemName=NS_WaterfallFX_B
 emitters [
  {emitterName=BottomSmokeA, rendererClasses [&NiagaraSpriteRendererProperties]}
  {emitterName=BottomSmokeB, rendererClasses [&NiagaraSpriteRendererProperties]}
  {emitterName=TopSmoke, rendererClasses [&NiagaraSpriteRendererProperties]}
  {emitterName=TopSplash, rendererClasses [&NiagaraSpriteRendererProperties]}
  {emitterName=DropSplash, rendererClasses [&NiagaraSpriteRendererProperties]}
  {emitterName=DropSmoke, rendererClasses [&NiagaraSpriteRendererProperties]}
  {emitterName=EdgeFlow, rendererClasses [&NiagaraSpriteRendererProperties]}
  {emitterName=DropParticle, rendererClasses [&NiagaraSpriteRendererProperties]}
  {emitterName=Fountain, rendererClasses [&NiagaraSpriteRendererProperties]}
  {emitterName=BottomRippleA, rendererClasses [&NiagaraSpriteRendererProperties]}
  {emitterName=BottomSplash, rendererClasses [&NiagaraSpriteRendererProperties]}
 ]
 userVariables [
  User.Bottom SPlash Scale {type=NiagaraFloat, defaultValue=1, description=""}
  User.BottomPosition {type=NiagaraPosition, description="", defaultValue {}}
  User.BottomSmokeDensity {type=NiagaraFloat, defaultValue=1, description=""}
  User.BottomSplash Color {type=LinearColor, description="", defaultValue {struct=LinearColor, value=<rgba 0.7 1 1 1>}}
  User.bottomsplashrate {type=NiagaraFloat, defaultValue=50, description=""}
  User.FallWidth {type=NiagaraFloat, defaultValue=900, description=""}
  User.HitSplashEffectDensity {type=NiagaraFloat, defaultValue=1, description=""}
  User.RainbowStrength {type=NiagaraFloat, defaultValue=1, description=""}
  User.RippleOffset {type=NiagaraFloat, defaultValue=70, description=""}
  User.RippleOpacity {type=NiagaraFloat, defaultValue=1, description=""}
  User.RippleWidth {type=NiagaraFloat, defaultValue=10000, description=""}
  User.SpawnSplashPositionArray {
   type=NiagaraDataInterfaceArrayPosition
   description=""
   defaultValue {
    struct=NiagaraExt_VariableValue_DataInterface
    value {
     dataInterface=&self:NiagaraDataInterfaceArrayPosition_0
     dataInterfaceClass=&NiagaraDataInterfaceArrayPosition
    }
   }
  }
  User.TopPosition {type=NiagaraPosition, description="", defaultValue {}}
  User.TopSplashRate {type=NiagaraFloat, defaultValue=20, description=""}
  User.TopSplashScale {type=NiagaraFloat, defaultValue=1, description=""}
  User.UseIdMask {type=NiagaraBool, defaultValue=true}
 ]
}
